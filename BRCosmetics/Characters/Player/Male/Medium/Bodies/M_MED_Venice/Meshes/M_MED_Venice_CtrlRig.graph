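# Control Rig for the M_MED_Venice male player body that applies pose-driven corrections to its hip pack and necklace bones.
# Spherical pose readers on both thighs, relative to the pelvis, weight offsets on the dyn_pack bones, with scaling on two left-driven ones.
# A second sequence branch nudges dyn_necklace_01 as dyn_necklace_03 nears spine_04_ctrl, using remapped distance as weight.
# The right-thigh reader has debug drawing enabled.
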ControlRigBlueprint /BRCosmetics/Characters/Player/Male/Medium/Bodies/M_MED_Venice/Meshes/M_MED_Venice_CtrlRig
graphs [
 RigVMFunctionLibrary {links [], nodes []}
 RigVMModel {nodes [
   RigUnit_BeginExecution = &RigUnit_BeginExecution()
   SequenceAggregate = &RigVMFunction_Sequence(ExecuteContext=@RigUnit_BeginExecution.ExecutePin)
   n1 = &RigUnit_SphericalPoseReader(
    ActiveRegionScaleFactors=(PositiveWidth=0.7,NegativeWidth=0.7,PositiveHeight=0.7,NegativeHeight=0.7)
    ActiveRegionSize=0.1
    Debug=(bDrawDebug=false,bDraw2D=false,DebugScale=25,DebugSegments=20,DebugThickness=0.25)
    DriverAxis=(X=-1,Y=0,Z=0)
    DriverItem=(Type=Bone,Name="thigh_l")
    FalloffRegionScaleFactors=(PositiveWidth=1,NegativeWidth=3,PositiveHeight=2,NegativeHeight=2)
    FalloffSize=0.215
    OptionalParentItem=(Type=Bone,Name="pelvis")
    RotationOffset=(X=100,Y=0,Z=-20)
    ExecutePin=@SequenceAggregate.A
   )
   > &RigUnit_ModifyTransforms(
    ItemToModify=((Item=(Type=Bone,Name="dyn_pack_main"),Transform=(Rotation=(X=0,Y=0,Z=0,W=1),Translation=(X=0,Y=0.5,Z=4),Scale3D=(X=1,Y=1,Z=1))))
    ItemToModify.0.Transform.Rotation=&RigVMFunction_MathQuaternionFromEuler(Euler=(X=-15,Y=-10,Z=0), RotationOrder=XZY).Result
    Weight=@n1.OutputParam
   )
   > &RigUnit_ModifyTransforms(
    ItemToModify=((Item=(Type=Bone,Name="dyn_pack_01"),Transform=(Rotation=(X=0,Y=0,Z=0,W=1),Translation=(X=0,Y=3,Z=5),Scale3D=(X=1.5,Y=1,Z=1))))
    ItemToModify.0.Transform.Rotation=&RigVMFunction_MathQuaternionFromEuler(Euler=(X=-5,Y=0,Z=0), RotationOrder=XZY).Result
    Weight=@n1.OutputParam
   )
   > &RigUnit_ModifyTransforms(
    ItemToModify=((Item=(Type=Bone,Name="dyn_pack_02"),Transform=(Rotation=(X=0,Y=0,Z=0,W=1),Translation=(X=0,Y=2,Z=5),Scale3D=(X=1.5,Y=0.9,Z=0.9))))
    ItemToModify.0.Transform.Rotation=&RigVMFunction_MathQuaternionFromEuler(Euler=(X=-5,Y=0,Z=0), RotationOrder=XZY).Result
    Weight=@n1.OutputParam
   )
   > n2 = &RigUnit_SphericalPoseReader(
    ActiveRegionScaleFactors=(PositiveWidth=0.7,NegativeWidth=0.6,PositiveHeight=0.7,NegativeHeight=0.7)
    ActiveRegionSize=0.1
    Debug=(bDrawDebug=true,bDraw2D=false,DebugScale=25,DebugSegments=20,DebugThickness=0.25)
    DriverAxis=(X=1,Y=0,Z=0)
    DriverItem=(Type=Bone,Name="thigh_r")
    FalloffRegionScaleFactors=(PositiveWidth=1,NegativeWidth=2.5,PositiveHeight=1.3,NegativeHeight=1)
    FalloffSize=0.2
    OptionalParentItem=(Type=Bone,Name="pelvis")
    RotationOffset=(X=-95,Y=180,Z=-10)
   )
   > &RigUnit_ModifyTransforms(
    ItemToModify=((Item=(Type=Bone,Name="dyn_pack_main"),Transform=(Rotation=(X=0,Y=0,Z=0,W=1),Translation=(X=0,Y=-1,Z=4),Scale3D=(X=1,Y=1,Z=1))))
    ItemToModify.0.Transform.Rotation=&RigVMFunction_MathQuaternionFromEuler(Euler=(X=35,Y=0,Z=0), RotationOrder=XZY).Result
    Weight=@n2.OutputParam
   )
   > &RigUnit_ModifyTransforms(
    ItemToModify=((Item=(Type=Bone,Name="dyn_pack_03"),Transform=(Rotation=(X=0,Y=0,Z=0,W=1),Translation=(X=-1,Y=0,Z=3),Scale3D=(X=1,Y=1,Z=1))))
    Weight=@n2.OutputParam
   )
   Remap = &RigVMFunction_MathDoubleRemap(
    SourceMaximum=4.3
    SourceMinimum=2
    TargetMaximum=0
    TargetMinimum=1
    Value=&RigVMFunction_MathVectorDistance(
     A=&RigUnit_GetTransform(Item=(Type=Bone,Name="dyn_necklace_03")).Transform.Translation
     B=&RigUnit_GetTransform(Item=(Type=Control,Name="spine_04_ctrl")).Transform.Translation
    ).Result
   )
   &RigUnit_ModifyTransforms(
    ItemToModify=((Item=(Type=Bone,Name="dyn_necklace_01"),Transform=(Rotation=(X=0,Y=0,Z=0,W=1),Translation=(X=0,Y=1,Z=1),Scale3D=(X=1,Y=1,Z=1))))
    ItemToModify.0.Transform.Rotation=&RigVMFunction_MathQuaternionFromEuler(Euler=(X=0,Y=0,Z=20), RotationOrder=XZY).Result
    Weight=@Remap.Result
    ExecutePin=@SequenceAggregate.B
   )
  ]}
]
hierarchy [
 root {loc=<0 0 0>, rot=<0 0 0 1>}
 pelvis {parent=root, loc=<0 -0.7071 93.1191>, rot=<0.0127 0.707 -0.0127 -0.707>}
 spine_01 {parent=pelvis, loc=<6.7494 -0.6388 0>, rot=<0 0 0.034 -0.9994>}
 spine_02 {parent=spine_01, loc=<7.2518 -0.035 0>, rot=<0 0 0.0238 -0.9997>}
 spine_03 {parent=spine_02, loc=<6.0104 0 0>, rot=<0 0 0.0149 -0.9999>}
 spine_04 {parent=spine_03, loc=<8.1024 0 0>, rot=<0 0 -0.0132 -0.9999>}
 spine_05 {parent=spine_04, loc=<10.0146 0 0>, rot=<0 0 -0.0015 -1>}
 neck_01 {parent=spine_05, loc=<11.3138 0.1529 0>, rot=<0 0 -0.2003 -0.9797>}
 neck_02 {parent=neck_01, loc=<4.5627 0 0>, rot=<0 0 0.0504 -0.9987>}
 head {parent=neck_02, loc=<5.3914 0 0>, rot=<0 0 0.115 -0.9934>}
 attach_cape {parent=spine_05, loc=<9.017 -11.714 0>, rot=<0.0616 -0.7044 -0.0616 -0.7044>}
 clavicle_l {parent=spine_05, loc=<4.3575 -0.1466 -2.3223>, rot=<0.0465 -0.7289 -0.009 -0.6829>}
 upperarm_l {parent=clavicle_l, loc=<15.5111 0 0>, rot=<0.0226 -0.3234 -0.0088 -0.9459>}
 lowerarm_l {parent=upperarm_l, loc=<26.6514 0 0>, rot=<0.0391 -0.0099 -0.2453 -0.9686>}
 hand_l {parent=lowerarm_l, loc=<22.3491 0 0>, rot=<-0.7081 -0.0507 0.0975 -0.6975>}
 thumb_01_l {parent=hand_l, loc=<1.3118 -0.3191 -2.459>, rot=<0.6546 -0.4042 0.0664 -0.6354>}
 thumb_02_l {parent=thumb_01_l, loc=<5.9336 -0.0448 0.0401>, rot=<-0.0037 0.0213 0.298 -0.9543>}
 thumb_03_l {parent=thumb_02_l, loc=<2.2223 0.3682 -0.0907>, rot=<0.0118 -0.013 0.0066 -0.9998>}
 index_metacarpal_l {parent=hand_l, loc=<1.6535 0.0183 -2.96>, rot=<0.0158 -0.0124 -0.0173 -0.9996>}
 index_01_l {parent=index_metacarpal_l, loc=<8.3008 0.0357 -0.0524>, rot=<-0.0044 0.0485 0.0924 -0.9945>}
 index_02_l {parent=index_01_l, loc=<3.3587 -0.3004 0.0452>, rot=<-0.002 0.0027 0.3225 -0.9466>}
 index_03_l {parent=index_02_l, loc=<2.5578 -0.123 -0.0252>, rot=<-0.002 0.0067 0.2558 -0.9667>}
 middle_metacarpal_l {parent=hand_l, loc=<2.1856 -0.1725 -0.5984>, rot=<0.006 0.0168 -0.0329 -0.9993>}
 middle_01_l {parent=middle_metacarpal_l, loc=<7.3275 -0.2447 0.0129>, rot=<-0.0208 0.0042 0.0858 -0.9961>}
 middle_02_l {parent=middle_01_l, loc=<4.0123 0.1401 0.1175>, rot=<-0.0218 0.0258 0.3512 -0.9357>}
 middle_03_l {parent=middle_02_l, loc=<2.6077 0.0534 0.1373>, rot=<0.0082 0.0188 0.1708 -0.9851>}
 ring_metacarpal_l {parent=hand_l, loc=<2.2074 -0.3625 1.3638>, rot=<-0.0284 0.0722 -0.0458 -0.9959>}
 ring_01_l {parent=ring_metacarpal_l, loc=<7.0295 -0.6197 -0.1446>, rot=<-0.0489 -0.0252 0.1892 -0.9804>}
 ring_02_l {parent=ring_01_l, loc=<3.6174 0.4647 -0.1027>, rot=<0 0 0.2806 -0.9598>}
 ring_03_l {parent=ring_02_l, loc=<2.4152 0.114 0.0199>, rot=<0 0 0.1189 -0.9929>}
 pinky_metacarpal_l {parent=hand_l, loc=<1.7616 -0.9753 3.3814>, rot=<-0.0397 0.0814 -0.069 -0.9935>}
 pinky_01_l {parent=pinky_metacarpal_l, loc=<7.1709 -0.4831 -0.1094>, rot=<-0.0654 -0.0035 0.2242 -0.9723>}
 pinky_02_l {parent=pinky_01_l, loc=<2.8274 0.0893 -0.1043>, rot=<0.0037 0.0011 0.2854 -0.9584>}
 pinky_03_l {parent=pinky_02_l, loc=<2.2046 0.0307 -0.0078>, rot=<-0.0002 0.0098 0.1445 -0.9895>}
 lowerarm_twist_02_l {parent=lowerarm_l, loc=<9.6199 0 0>, rot=<0 0 0 -1>}
 lowerarm_twist_01_l {parent=lowerarm_l, loc=<17.8549 0 0>, rot=<0 0 0 -1>}
 dyn_armstrap_L_01_mstr {parent=lowerarm_l, loc=<1.6402 -3.245 3.3679>, rot=<-0.0465 0.0111 0.2443 -0.9685>}
 dyn_armstrap_L_01_01 {parent=dyn_armstrap_L_01_mstr, loc=<0.8219 0 0>, rot=<0 0 0 1>}
 dyn_armstrap_L_01_02 {parent=dyn_armstrap_L_01_01, loc=<5.5455 0 0>, rot=<0.0006 -0.0184 -0.0028 -0.9998>}
 dyn_armstrap_L_01_03 {parent=dyn_armstrap_L_01_02, loc=<7.2433 0 0>, rot=<0.0007 -0.0238 0.0029 -0.9997>}
 dyn_armstrap_L_01_04 {parent=dyn_armstrap_L_01_03, loc=<9.0632 0 0>, rot=<-0.0009 0.0241 0.0015 -0.9997>}
 dyn_armstrap_L_01_05 {parent=dyn_armstrap_L_01_04, loc=<7.9881 0 0>, rot=<0.0007 -0.0195 -0.0026 -0.9998>}
 dyn_armstrap_L_01_06 {parent=dyn_armstrap_L_01_05, loc=<6.8475 0 0>, rot=<0 0 0 1>}
 dyn_armstrap_L_01_07 {parent=dyn_armstrap_L_01_06, loc=<8.1074 0 0>, rot=<0 0 0 1>}
 dyn_armstrap_L_02_mstr {parent=lowerarm_l, loc=<3.3153 -4.679 2.2163>, rot=<-0.0529 -0.0167 0.2274 -0.9722>}
 dyn_armstrap_L_02_01 {parent=dyn_armstrap_L_02_mstr, loc=<1.3957 0 0>, rot=<0 0 0 1>}
 dyn_armstrap_L_02_02 {parent=dyn_armstrap_L_02_01, loc=<6.1331 0 0>, rot=<0.0001 -0.0236 0.0007 -0.9997>}
 dyn_armstrap_L_02_03 {parent=dyn_armstrap_L_02_02, loc=<8.0373 0 0>, rot=<0 -0.0045 0.0067 -1>}
 dyn_armstrap_L_02_04 {parent=dyn_armstrap_L_02_03, loc=<9.2937 0 0>, rot=<-0.0001 0.007 -0.0033 -1>}
 dyn_armstrap_L_02_05 {parent=dyn_armstrap_L_02_04, loc=<7.9396 0 0>, rot=<0.0001 -0.0083 -0.0024 -1>}
 dyn_armstrap_L_02_06 {parent=dyn_armstrap_L_02_05, loc=<8.2926 0 0>, rot=<0 0 0 1>}
 dyn_armstrap_L_02_07 {parent=dyn_armstrap_L_02_06, loc=<6.3936 0 0>, rot=<0 0 0 1>}
 dyn_armstrap_L_03_mstr {parent=lowerarm_l, loc=<-4.0937 -5.8465 1.0032>, rot=<-0.0545 -0.0277 0.2515 -0.9659>}
 dyn_armstrap_L_03_01 {parent=dyn_armstrap_L_03_mstr, loc=<0 0 0>, rot=<0 0 0 1>}
 dyn_armstrap_L_03_02 {parent=dyn_armstrap_L_03_01, loc=<3.5622 0 0>, rot=<0.001 -0.0261 -0.0132 -0.9996>}
 dyn_armstrap_L_03_03 {parent=dyn_armstrap_L_03_02, loc=<5.0894 0 0>, rot=<-0.0003 0.0137 -0.0029 -0.9999>}
 dyn_armstrap_L_03_04 {parent=dyn_armstrap_L_03_03, loc=<6.8305 0 0>, rot=<0.0005 -0.0339 -0.0055 -0.9994>}
 dyn_armstrap_L_03_05 {parent=dyn_armstrap_L_03_04, loc=<8.654 0 0>, rot=<0 0 0 1>}
 dyn_armstrap_L_03_06 {parent=dyn_armstrap_L_03_05, loc=<5.2658 0 0>, rot=<0 0 0 1>}
 upperarm_twist_01_l {parent=upperarm_l, loc=<0.0041 0 0>, rot=<0 0 0 -1>}
 deform_upperarm_bk_base_l {parent=upperarm_twist_01_l, loc=<1.0813 -7.2193 -1.9453>, rot=<0 0 0 -1>}
 deform_upperarm_bk_l {parent=deform_upperarm_bk_base_l, loc=<0 0 0>, rot=<0 0 0 1>}
 deform_upperarm_fr_base_l {parent=upperarm_twist_01_l, loc=<-0.3691 7.0506 -2.765>, rot=<0 0 0 -1>}
 deform_upperarm_fr_l {parent=deform_upperarm_fr_base_l, loc=<0 0 0>, rot=<0 0 0 1>}
 deform_upperarm_up_base_l {parent=upperarm_twist_01_l, loc=<-3.2216 -1.1391 3.3311>, rot=<0 0 0 -1>}
 deform_upperarm_up_l {parent=deform_upperarm_up_base_l, loc=<0 0 0>, rot=<0 0 0 1>}
 upperarm_twist_02_l {parent=upperarm_l, loc=<13.7941 0 0>, rot=<0 0 0 -1>}
 dyn_armstrap_L_04_mstr {parent=upperarm_twist_02_l, loc=<7.0915 4.8864 -3.6739>, rot=<0.6874 -0.2282 -0.3073 -0.6172>}
 dyn_armstrap_L_04_01 {parent=dyn_armstrap_L_04_mstr, loc=<0 0 0>, rot=<0 0 0 1>}
 dyn_armstrap_L_04_02 {parent=dyn_armstrap_L_04_01, loc=<4.4578 0 0>, rot=<-0.0005 -0.0306 0.0149 -0.9994>}
 dyn_armstrap_L_04_03 {parent=dyn_armstrap_L_04_01, loc=<2.2317 1.0259 -0.3146>, rot=<-0.7128 -0.2274 -0.1254 -0.6515>}
 attach_backpack {parent=spine_05, loc=<-0.123 -13.473 0>, rot=<-0.0093 -0.707 0.0093 -0.707>}
 clavicle_r {parent=spine_05, loc=<4.3575 -0.1466 2.3223>, rot=<0.7289 0.0465 -0.6829 0.009>}
 upperarm_r {parent=clavicle_r, loc=<-15.5111 0 0>, rot=<0.0226 -0.3234 -0.0088 -0.9459>}
 lowerarm_r {parent=upperarm_r, loc=<-26.6514 0 0>, rot=<0.0391 -0.0099 -0.2453 -0.9686>}
 hand_r {parent=lowerarm_r, loc=<-22.3491 0 0>, rot=<-0.7081 -0.0507 0.0975 -0.6975>}
 thumb_01_r {parent=hand_r, loc=<-1.3118 0.3191 2.459>, rot=<0.6546 -0.4042 0.0664 -0.6354>}
 thumb_02_r {parent=thumb_01_r, loc=<-5.9336 0.0448 -0.0401>, rot=<-0.0037 0.0213 0.298 -0.9543>}
 thumb_03_r {parent=thumb_02_r, loc=<-2.2223 -0.3682 0.0907>, rot=<0.0118 -0.013 0.0066 -0.9998>}
 index_metacarpal_r {parent=hand_r, loc=<-1.6535 -0.0183 2.96>, rot=<0.0158 -0.0124 -0.0173 -0.9996>}
 index_01_r {parent=index_metacarpal_r, loc=<-8.3008 -0.0357 0.0524>, rot=<-0.0044 0.0485 0.0924 -0.9945>}
 index_02_r {parent=index_01_r, loc=<-3.3587 0.3004 -0.0452>, rot=<-0.002 0.0027 0.3225 -0.9466>}
 index_03_r {parent=index_02_r, loc=<-2.5578 0.123 0.0252>, rot=<-0.002 0.0067 0.2558 -0.9667>}
 middle_metacarpal_r {parent=hand_r, loc=<-2.1856 0.1725 0.5984>, rot=<0.006 0.0168 -0.0329 -0.9993>}
 middle_01_r {parent=middle_metacarpal_r, loc=<-7.3275 0.2447 -0.0129>, rot=<-0.0208 0.0042 0.0858 -0.9961>}
 middle_02_r {parent=middle_01_r, loc=<-4.0123 -0.1401 -0.1175>, rot=<-0.0218 0.0258 0.3512 -0.9357>}
 middle_03_r {parent=middle_02_r, loc=<-2.6077 -0.0534 -0.1373>, rot=<0.0082 0.0188 0.1708 -0.9851>}
 ring_metacarpal_r {parent=hand_r, loc=<-2.2074 0.3625 -1.3638>, rot=<-0.0284 0.0722 -0.0458 -0.9959>}
 ring_01_r {parent=ring_metacarpal_r, loc=<-7.0295 0.6197 0.1446>, rot=<-0.0489 -0.0252 0.1892 -0.9804>}
 ring_02_r {parent=ring_01_r, loc=<-3.6174 -0.4647 0.1027>, rot=<0 0 0.2806 -0.9598>}
 ring_03_r {parent=ring_02_r, loc=<-2.4152 -0.114 -0.0199>, rot=<0 0 0.1189 -0.9929>}
 pinky_metacarpal_r {parent=hand_r, loc=<-1.7616 0.9753 -3.3814>, rot=<-0.0397 0.0814 -0.069 -0.9935>}
 pinky_01_r {parent=pinky_metacarpal_r, loc=<-7.1709 0.4831 0.1094>, rot=<-0.0654 -0.0035 0.2242 -0.9723>}
 pinky_02_r {parent=pinky_01_r, loc=<-2.8274 -0.0893 0.1043>, rot=<0.0037 0.0011 0.2854 -0.9584>}
 pinky_03_r {parent=pinky_02_r, loc=<-2.2046 -0.0307 0.0078>, rot=<-0.0002 0.0098 0.1445 -0.9895>}
 lowerarm_twist_02_r {parent=lowerarm_r, loc=<-9.6199 0 0>, rot=<0 0 0 -1>}
 lowerarm_twist_01_r {parent=lowerarm_r, loc=<-17.8549 0 0>, rot=<0 0 0 -1>}
 dyn_armstrap_R_03_mstr {parent=lowerarm_r, loc=<4.8159 4.0712 -2.7678>, rot=<-0.2269 -0.972 -0.0551 0.0262>}
 dyn_armstrap_R_03_01 {parent=dyn_armstrap_R_03_mstr, loc=<0 0 0>, rot=<0.1146 0.013 -0.1119 -0.987>}
 dyn_armstrap_R_03_02 {parent=dyn_armstrap_R_03_01, loc=<3.6344 0 0>, rot=<-0.1217 -0.0389 0.1138 -0.9853>}
 dyn_armstrap_R_03_03 {parent=dyn_armstrap_R_03_02, loc=<5.0914 0 0>, rot=<0.0158 0.0138 -0.014 -0.9997>}
 dyn_armstrap_R_03_04 {parent=dyn_armstrap_R_03_03, loc=<6.8305 0 0>, rot=<-0.0058 -0.0032 0.0053 -1>}
 dyn_armstrap_R_03_05 {parent=dyn_armstrap_R_03_04, loc=<8.6416 0 0>, rot=<-0.0009 0.022 0.0008 -0.9998>}
 dyn_armstrap_R_03_06 {parent=dyn_armstrap_R_03_05, loc=<6.4554 0 0>, rot=<0 0 0 1>}
 dyn_armstrap_R_01_mstr {parent=lowerarm_r, loc=<-4.2385 5.0279 -1.5416>, rot=<-0.2269 -0.972 -0.0551 0.0262>}
 dyn_armstrap_R_01_01 {parent=dyn_armstrap_R_01_mstr, loc=<0.819 0.0262 0.0642>, rot=<0.0089 -0.0167 -0.0073 -0.9998>}
 dyn_armstrap_R_01_02 {parent=dyn_armstrap_R_01_01, loc=<5.5455 0 0>, rot=<0.0131 -0.0181 -0.0113 -0.9997>}
 dyn_armstrap_R_01_03 {parent=dyn_armstrap_R_01_02, loc=<7.2435 0 0>, rot=<-0.0036 0.0266 0.0031 -0.9996>}
 dyn_armstrap_R_01_04 {parent=dyn_armstrap_R_01_03, loc=<9.0653 0 0>, rot=<0.0067 0.0127 -0.0064 -0.9999>}
 dyn_armstrap_R_01_05 {parent=dyn_armstrap_R_01_04, loc=<8.0385 0 0>, rot=<-0.0199 0.0018 0.0195 -0.9996>}
 dyn_armstrap_R_01_06 {parent=dyn_armstrap_R_01_05, loc=<6.8993 0 0>, rot=<-0.0046 -0.026 0.0016 -0.9997>}
 dyn_armstrap_R_01_07 {parent=dyn_armstrap_R_01_06, loc=<5.7621 0 0>, rot=<0 0 0 1>}
 dyn_armstrap_R_02_mstr {parent=lowerarm_r, loc=<-4.511 5.302 0.0963>, rot=<-0.2269 -0.972 -0.0551 0.0262>}
 dyn_armstrap_R_02_01 {parent=dyn_armstrap_R_02_mstr, loc=<1.396 0 0>, rot=<-0.0001 0.0269 0 -0.9996>}
 dyn_armstrap_R_02_02 {parent=dyn_armstrap_R_02_01, loc=<6.1374 0 -0.6582>, rot=<-0.0138 -0.0608 0.0141 -0.998>}
 dyn_armstrap_R_02_03 {parent=dyn_armstrap_R_02_02, loc=<8.0028 0 0>, rot=<0.0129 -0.0014 -0.0108 -0.9999>}
 dyn_armstrap_R_02_04 {parent=dyn_armstrap_R_02_03, loc=<9.3014 -0.0796 -0.0033>, rot=<-0.0208 -0.0135 0.0168 -0.9996>}
 dyn_armstrap_R_02_05 {parent=dyn_armstrap_R_02_04, loc=<7.9521 0.1532 -0.0089>, rot=<0.029 0.0045 -0.023 -0.9993>}
 dyn_armstrap_R_02_06 {parent=dyn_armstrap_R_02_05, loc=<8.3029 0 0>, rot=<-0.0061 0.0248 0.0025 -0.9997>}
 dyn_armstrap_R_02_07 {parent=dyn_armstrap_R_02_06, loc=<4.9388 0 0>, rot=<0 0 0 1>}
 upperarm_twist_01_r {parent=upperarm_r, loc=<-0.0041 0 0>, rot=<0 0 0 -1>}
 deform_upperarm_bk_base_r {parent=upperarm_twist_01_r, loc=<-1.0814 7.2192 1.9455>, rot=<0 0 0 1>}
 deform_upperarm_bk_r {parent=deform_upperarm_bk_base_r, loc=<0 0 0>, rot=<0 0 0 1>}
 deform_upperarm_fr_base_r {parent=upperarm_twist_01_r, loc=<0.3689 -7.0506 2.7652>, rot=<0 0 0 1>}
 deform_upperarm_fr_r {parent=deform_upperarm_fr_base_r, loc=<0 0 0>, rot=<0 0 0 1>}
 deform_upperarm_up_base_r {parent=upperarm_twist_01_r, loc=<3.2214 1.1391 -3.3308>, rot=<0 0 0 1>}
 deform_upperarm_up_r {parent=deform_upperarm_up_base_r, loc=<0 0 0>, rot=<0 0 0 -1>}
 upperarm_twist_02_r {parent=upperarm_r, loc=<-13.7941 0 0>, rot=<0 0 0 -1>}
 dyn_armstrap_R_04_mstr {parent=upperarm_twist_02_r, loc=<-3.2655 -6.1653 3.0806>, rot=<0.4748 0.0172 0.8799 0.0092>}
 dyn_armstrap_R_04_01 {parent=dyn_armstrap_R_04_mstr, loc=<0 0 0>, rot=<0.7071 0 0 -0.7071>}
 dyn_armstrap_R_04_02 {parent=dyn_armstrap_R_04_01, loc=<3.6987 1.4368 0.38>, rot=<-0.064 0.0243 -0.1164 -0.9908>}
 dyn_armstrap_R_04_03 {parent=dyn_armstrap_R_04_01, loc=<2.8763 0 0>, rot=<-0.1076 0 0 -0.9942>}
 dyn_necklace_mstr {parent=spine_05, loc=<8.418 6.3356 -0.0477>, rot=<0.0284 -0.7065 -0.0284 -0.7065>}
 dyn_necklace_01 {parent=dyn_necklace_mstr, loc=<0 0 0>, rot=<-0.1899 -0.6811 -0.1899 -0.6811>}
 dyn_necklace_02 {parent=dyn_necklace_01, loc=<6.1848 0 0>, rot=<0 0 0 1>}
 dyn_necklace_03 {parent=dyn_necklace_02, loc=<4.9123 0 0>, rot=<0 0 0 1>}
 deform_pec_aim_base_l {parent=spine_04, loc=<7.8732 12.6388 -2.4123>, rot=<0.0102 -0.4602 0.3979 -0.7936>}
 deform_pec_aim_l {parent=deform_pec_aim_base_l, loc=<0 0 0>, rot=<0 0 0 1>}
 deform_pec_base_l {parent=deform_pec_aim_l, loc=<10.766 0 0>, rot=<0 0 0 1>}
 deform_pec_l {parent=deform_pec_base_l, loc=<0.055 0 0>, rot=<0 0 0 1>}
 deform_pec_aim_base_r {parent=spine_04, loc=<7.8727 12.6388 2.4123>, rot=<-0.4602 -0.0102 0.7936 0.3979>}
 deform_pec_aim_r {parent=deform_pec_aim_base_r, loc=<0 0 0>, rot=<0 0 0 1>}
 deform_pec_base_r {parent=deform_pec_aim_r, loc=<-10.7662 -0.0006 -0.0007>, rot=<0 0 0 1>}
 deform_pec_r {parent=deform_pec_base_r, loc=<0 0 0>, rot=<0 0 0 1>}
 spine_04_ctrl {controlType=EULER_TRANSFORM, parent=spine_04, type=Control, loc=<0 0 0>, rot=<0 0 0 1>}
 thigh_l {parent=pelvis, loc=<-2.5693 -0.0109 -7.7501>, rot=<0.0012 0.0572 0.0155 -0.9982>}
 calf_l {parent=thigh_l, loc=<-39.0484 0 0>, rot=<0.0006 -0.0161 -0.0391 -0.9991>}
 foot_l {parent=calf_l, loc=<-42.561 0.1077 0.4292>, rot=<0.0418 -0.0455 0.0343 -0.9975>}
 ball_l {parent=foot_l, loc=<-6.1599 13.8264 0.6806>, rot=<0.0001 -0.0001 -0.7063 -0.7079>}
 calf_twist_02_l {parent=calf_l, loc=<-18.8064 -0.0967 0.0397>, rot=<0.0237 0.0095 -0.0035 -0.9997>}
 calf_twist_01_l {parent=calf_l, loc=<-35.813 -0.0967 0.0397>, rot=<0.0237 0.0095 -0.0035 -0.9997>}
 dyn_lace_l_01_mstr {parent=calf_twist_01_l, loc=<-3.981 5.3475 -3.8065>, rot=<-0.0498 -0.7417 0.0123 -0.6688>}
 dyn_lace_l_01_01 {parent=dyn_lace_l_01_mstr, loc=<0 0 0>, rot=<-0.0243 -0.5831 -0.0338 -0.8113>}
 dyn_lace_l_01_02 {parent=dyn_lace_l_01_01, loc=<2.6618 0 0>, rot=<-0.0005 -0.0212 0.0575 -0.9981>}
 dyn_lace_l_01_03 {parent=dyn_lace_l_01_02, loc=<2.6554 0 0>, rot=<0 0 0 1>}
 dyn_lace_l_02_mstr {parent=calf_twist_01_l, loc=<-3.7751 4.7212 4.6752>, rot=<0.7337 0.1195 0.6486 0.1635>}
 dyn_lace_l_02_01 {parent=dyn_lace_l_02_mstr, loc=<0 0 0>, rot=<-0.0147 -0.6479 -0.0188 -0.7614>}
 dyn_lace_l_02_02 {parent=dyn_lace_l_02_01, loc=<1.8435 0 0>, rot=<0.0455 -0.0502 -0.0405 -0.9969>}
 dyn_lace_l_02_03 {parent=dyn_lace_l_02_02, loc=<3.3083 0 0>, rot=<0.0049 0 0 -1>}
 thigh_twist_01_l {parent=thigh_l, loc=<-21.6303 0 0>, rot=<0.0013 0 0 -1>}
 dyn_fold_mstr {parent=thigh_twist_01_l, loc=<-13.5091 6.9669 -3.4738>, rot=<-0.4016 -0.8589 -0.3107 -0.0666>}
 dyn_fold {parent=dyn_fold_mstr, loc=<0 0 0>, rot=<0 0 0 1>}
 deform_groin_base_l {parent=thigh_l, loc=<-2.1908 8.1142 -1.7442>, rot=<0 0 0 -1>}
 deform_groin_l {parent=deform_groin_base_l, loc=<0 0 0>, rot=<0 0 0 1>}
 thigh_r {parent=pelvis, loc=<-2.5693 -0.0109 7.7501>, rot=<0.0572 -0.0012 0.9982 0.0155>}
 calf_r {parent=thigh_r, loc=<39.0484 0 0>, rot=<0.0006 -0.0161 -0.0391 -0.9991>}
 foot_r {parent=calf_r, loc=<42.561 -0.1077 -0.4292>, rot=<0.0418 -0.0455 0.0343 -0.9975>}
 ball_r {parent=foot_r, loc=<6.1599 -13.8264 -0.6806>, rot=<0.0001 -0.0001 -0.7063 -0.7079>}
 calf_twist_02_r {parent=calf_r, loc=<18.8064 0.0967 -0.0397>, rot=<0.0237 0.0095 -0.0035 -0.9997>}
 calf_twist_01_r {parent=calf_r, loc=<35.813 0.0967 -0.0397>, rot=<0.0237 0.0095 -0.0035 -0.9997>}
 dyn_lace_r_01_mstr {parent=calf_twist_01_r, loc=<3.981 -5.3475 3.8065>, rot=<-0.6688 0.0123 0.7417 0.0498>}
 dyn_lace_r_01_01 {parent=dyn_lace_r_01_mstr, loc=<0 0 0>, rot=<-0.5831 -0.0243 0.8113 0.0338>}
 dyn_lace_r_01_02 {parent=dyn_lace_r_01_01, loc=<2.6618 0 0>, rot=<0.0005 -0.0212 -0.0575 -0.9981>}
 dyn_lace_r_01_03 {parent=dyn_lace_r_01_02, loc=<2.6554 0 0>, rot=<0.0007 0.0411 0.016 -0.999>}
 dyn_lace_r_02_mstr {parent=calf_twist_01_r, loc=<3.7751 -4.7212 -4.6752>, rot=<-0.6688 0.0123 0.7417 0.0498>}
 dyn_lace_r_02_01 {parent=dyn_lace_r_02_mstr, loc=<0 0 0>, rot=<-0.009 -0.6462 -0.0106 -0.763>}
 dyn_lace_r_02_02 {parent=dyn_lace_r_02_01, loc=<1.8435 0 0>, rot=<0.0009 -0.0303 0.057 -0.9979>}
 dyn_lace_r_02_03 {parent=dyn_lace_r_02_02, loc=<3.3083 0 0>, rot=<0.0074 0.0059 0.0263 -0.9996>}
 thigh_twist_01_r {parent=thigh_r, loc=<21.6303 0 0>, rot=<0.0013 0 0 -1>}
 deform_groin_base_r {parent=thigh_r, loc=<2.1908 -8.1142 1.7442>, rot=<0 0 0 -1>}
 deform_groin_r {parent=deform_groin_base_r, loc=<0 0 0>, rot=<0 0 0 1>}
 dyn_pack_main {parent=pelvis, loc=<-2.2536 11.5238 -3.0473>, rot=<-0.4095 -0.4605 -0.5327 -0.5801>}
 dyn_pack_01 {parent=dyn_pack_main, loc=<-1.2022 -7.537 -2.323>, rot=<0 0 0 1>}
 dyn_pack_02 {parent=dyn_pack_main, loc=<0.6875 -1.7833 -3.9619>, rot=<0 0 0 1>}
 dyn_tassel_mstr {parent=dyn_pack_main, loc=<3.0953 7.2153 1.0644>, rot=<-0.1003 -0.0726 0.6676 -0.7341>}
 dyn_tassel_01 {parent=dyn_tassel_mstr, loc=<0 0 0>, rot=<-0.4931 0.3734 0.6528 0.4373>}
 dyn_tassel_02 {parent=dyn_tassel_01, loc=<6.4642 -0.6016 -0.1269>, rot=<-0.1655 -0.002 0.0474 -0.9851>}
 dyn_tassel_03 {parent=dyn_tassel_01, loc=<6.35 1.2194 -0.6259>, rot=<-0.0095 -0.0815 -0.074 -0.9939>}
 dyn_pack_03 {parent=dyn_pack_main, loc=<-0.3213 7.3428 -1.2861>, rot=<0 0 0 1>}
 deform_glute_base_l {parent=pelvis, loc=<-8.8064 -8.8222 -8.4287>, rot=<0.0275 0.0501 -0.4472 -0.8926>}
 deform_glute_l {parent=deform_glute_base_l, loc=<0 0 0>, rot=<0 0 0 1>}
 deform_glute_base_r {parent=pelvis, loc=<-8.8062 -8.8222 8.4287>, rot=<0.0501 -0.0275 0.8926 -0.4472>}
 deform_glute_r {parent=deform_glute_base_r, loc=<0 0 0>, rot=<0 0 0 1>}
]
