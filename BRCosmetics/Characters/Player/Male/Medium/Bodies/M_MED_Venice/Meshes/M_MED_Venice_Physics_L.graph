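# Physics asset for the left side of the medium male Venice player body, simulating three dangling straps on the left forearm.
# Capsules on pelvis, upper spine, left arm and left leg form a coarse limited-joint body.
# Each strap has a small master capsule, free-rotating on lowerarm_l, leading a four-link capsule chain.
# Chain joints allow wide swing but very tight twist, so straps bend without spinning.
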PhysicsAsset /BRCosmetics/Characters/Player/Male/Medium/Bodies/M_MED_Venice/Meshes/M_MED_Venice_Physics_L
bodies {
 calf_l [{shapeType=Capsule, length=19.5767, radius=10.7158, center=<-9.5354 0 0>, rotation=<rot 90 0 0>}]
 dyn_armstrap_L_01_01 [{shapeType=Capsule, length=5, radius=2, shapeName=dyn_armstrap_L_01_01_capsule, center=<5 0 0>, rotation=<rot 90 0 0>}]
 dyn_armstrap_L_01_02 [{shapeType=Capsule, length=6, radius=1, shapeName=dyn_armstrap_L_01_02_capsule, center=<5 0 0>, rotation=<rot 90 0 0>}]
 dyn_armstrap_L_01_03 [{shapeType=Capsule, length=5, radius=1.5, shapeName=dyn_armstrap_L_01_03_capsule, center=<5 0 0>, rotation=<rot 90 0 0>}]
 dyn_armstrap_L_01_04 [{shapeType=Capsule, length=5, radius=1.5, shapeName=dyn_armstrap_L_01_03_capsule, center=<5 0 0>, rotation=<rot 90 0 0>}]
 dyn_armstrap_L_01_mstr [{shapeType=Capsule, length=0, radius=0.5, shapeName=dyn_armstrap_L_01_mstr_capsule, rotation=<rot 90 0 0>}]
 dyn_armstrap_L_02_01 [{shapeType=Capsule, length=5, radius=2, shapeName=dyn_armstrap_L_01_01_capsule, center=<5 0 0>, rotation=<rot 90 0 0>}]
 dyn_armstrap_L_02_02 [{shapeType=Capsule, length=6, radius=1, shapeName=dyn_armstrap_L_01_02_capsule, center=<5 0 0>, rotation=<rot 90 0 0>}]
 dyn_armstrap_L_02_03 [{shapeType=Capsule, length=5, radius=1.5, shapeName=dyn_armstrap_L_01_03_capsule, center=<5 0 0>, rotation=<rot 90 0 0>}]
 dyn_armstrap_L_02_04 [{shapeType=Capsule, length=5, radius=1.5, shapeName=dyn_armstrap_L_01_03_capsule, center=<5 0 0>, rotation=<rot 90 0 0>}]
 dyn_armstrap_L_02_mstr [{shapeType=Capsule, length=0, radius=0.5, shapeName=dyn_armstrap_L_01_mstr_capsule, rotation=<rot 90 0 0>}]
 dyn_armstrap_L_03_01 [{shapeType=Capsule, length=5, radius=2, shapeName=dyn_armstrap_L_01_01_capsule, center=<5 0 0>, rotation=<rot 90 0 0>}]
 dyn_armstrap_L_03_02 [{shapeType=Capsule, length=5, radius=2, shapeName=dyn_armstrap_L_01_02_capsule, center=<5 0 0>, rotation=<rot 90 0 0>}]
 dyn_armstrap_L_03_03 [{shapeType=Capsule, length=5, radius=1.5, shapeName=dyn_armstrap_L_01_03_capsule, center=<5 0 0>, rotation=<rot 90 0 0>}]
 dyn_armstrap_L_03_04 [{shapeType=Capsule, length=5, radius=1.5, shapeName=dyn_armstrap_L_01_03_capsule, center=<5 0 0>, rotation=<rot 90 0 0>}]
 dyn_armstrap_L_03_mstr [{shapeType=Capsule, length=0, radius=0.5, shapeName=dyn_armstrap_L_01_mstr_capsule, rotation=<rot 90 0 0>}]
 foot_l [{shapeType=Capsule, length=17.07, radius=10.7523, center=<-1.927 7.7943 -0.2502>, rotation=<rot 0 12.1107 -87.744>}]
 hand_l [{shapeType=Capsule, length=8.6852, radius=6.1455, center=<8.2293 -2.007 0.2791>, rotation=<rot 0 76.4147 -74.5755>}]
 lowerarm_l [{shapeType=Capsule, length=17.8638, radius=4.6109, center=<10.7783 0 0>, rotation=<rot 90 0 0>}]
 pelvis [{shapeType=Capsule, length=34.0116, radius=16.7244, rotation=<rot 90 0 0>}]
 spine_05 [{shapeType=Capsule, length=19.0864, radius=12.939, center=<1.1702 -1.2832 0.7067>, rotation=<rot 0 -135.5378 -177.0553>}]
 thigh_l [{shapeType=Capsule, length=21.7461, radius=11.44, center=<-14.9256 0 0>, rotation=<rot 90 0 0>}]
 upperarm_l [{shapeType=Capsule, length=15.8448, radius=9.1249, center=<12.5514 -1.6258 -0.9539>, rotation=<rot 0 -85.1003 -86.9037>}]
}
constraints [
 {bone1Name=spine_05, bone2Name=pelvis, swing1LimitDegrees=45, swing2LimitDegrees=45, twistLimitDegrees=45}
 {bone1Name=upperarm_l, bone2Name=spine_05, swing1LimitDegrees=45, swing2LimitDegrees=45, twistLimitDegrees=45}
 {bone1Name=lowerarm_l, bone2Name=upperarm_l, swing1LimitDegrees=45, swing2LimitDegrees=45, twistLimitDegrees=45}
 {bone1Name=hand_l, bone2Name=lowerarm_l, swing1LimitDegrees=45, swing2LimitDegrees=45, twistLimitDegrees=45}
 {bone1Name=thigh_l, bone2Name=pelvis, swing1LimitDegrees=45, swing2LimitDegrees=45, twistLimitDegrees=45}
 {bone1Name=calf_l, bone2Name=thigh_l, swing1LimitDegrees=45, swing2LimitDegrees=45, twistLimitDegrees=45}
 {bone1Name=foot_l, bone2Name=calf_l, swing1LimitDegrees=45, swing2LimitDegrees=45, twistLimitDegrees=45}
 {
  bone1Name=dyn_armstrap_L_01_mstr
  bone2Name=lowerarm_l
  swing1LimitDegrees=45
  swing1Motion=Free
  swing2LimitDegrees=45
  swing2Motion=Free
  twistLimitDegrees=45
  twistMotion=Free
 }
 {bone1Name=dyn_armstrap_L_01_01, bone2Name=dyn_armstrap_L_01_mstr, swing1LimitDegrees=35, swing2LimitDegrees=45, twistLimitDegrees=2}
 {bone1Name=dyn_armstrap_L_01_02, bone2Name=dyn_armstrap_L_01_01, swing1LimitDegrees=65, swing2LimitDegrees=65, twistLimitDegrees=2}
 {bone1Name=dyn_armstrap_L_01_03, bone2Name=dyn_armstrap_L_01_02, swing1LimitDegrees=65, swing2LimitDegrees=65, twistLimitDegrees=2}
 {
  bone1Name=dyn_armstrap_L_02_mstr
  bone2Name=lowerarm_l
  swing1LimitDegrees=45
  swing1Motion=Free
  swing2LimitDegrees=45
  swing2Motion=Free
  twistLimitDegrees=45
  twistMotion=Free
 }
 {bone1Name=dyn_armstrap_L_02_01, bone2Name=dyn_armstrap_L_02_mstr, swing1LimitDegrees=55, swing2LimitDegrees=45, twistLimitDegrees=2}
 {bone1Name=dyn_armstrap_L_02_02, bone2Name=dyn_armstrap_L_02_01, swing1LimitDegrees=65, swing2LimitDegrees=65, twistLimitDegrees=2}
 {bone1Name=dyn_armstrap_L_02_03, bone2Name=dyn_armstrap_L_02_02, swing1LimitDegrees=65, swing2LimitDegrees=65, twistLimitDegrees=2}
 {
  bone1Name=dyn_armstrap_L_03_mstr
  bone2Name=lowerarm_l
  swing1LimitDegrees=45
  swing1Motion=Free
  swing2LimitDegrees=45
  swing2Motion=Free
  twistLimitDegrees=45
  twistMotion=Free
 }
 {bone1Name=dyn_armstrap_L_03_01, bone2Name=dyn_armstrap_L_03_mstr, swing1LimitDegrees=50, swing2LimitDegrees=55, twistLimitDegrees=2}
 {bone1Name=dyn_armstrap_L_03_02, bone2Name=dyn_armstrap_L_03_01, swing1LimitDegrees=65, swing2LimitDegrees=65, twistLimitDegrees=2}
 {bone1Name=dyn_armstrap_L_03_03, bone2Name=dyn_armstrap_L_03_02, swing1LimitDegrees=65, swing2LimitDegrees=65, twistLimitDegrees=2}
 {bone1Name=dyn_armstrap_L_01_04, bone2Name=dyn_armstrap_L_01_03, swing1LimitDegrees=65, swing2LimitDegrees=65, twistLimitDegrees=2}
 {bone1Name=dyn_armstrap_L_02_04, bone2Name=dyn_armstrap_L_02_03, swing1LimitDegrees=65, swing2LimitDegrees=65, twistLimitDegrees=2}
 {bone1Name=dyn_armstrap_L_03_04, bone2Name=dyn_armstrap_L_03_03, swing1LimitDegrees=65, swing2LimitDegrees=65, twistLimitDegrees=2}
]
