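# Right-side physics asset for the M_MED_Venice male player cosmetic body, set up for three dynamic arm strap chains.
# Capsules on pelvis, upper spine, right arm and right leg form a partial body, with limb joints left angularly free.
# Each strap hangs off lowerarm_r through a free-rotating master body, then four linked capsules with limited swing and very tight twist.
# Strap shape names carry left-side labels despite being on right-side bones.
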PhysicsAsset /BRCosmetics/Characters/Player/Male/Medium/Bodies/M_MED_Venice/Meshes/M_MED_Venice_Physics_R
bodies {
 calf_r [{shapeType=Capsule, length=9.8869, radius=7.5792, shapeName=calf_r_capsule, center=<6.3717 0.7037 -0.3585>, rotation=<rot 0 16.0148 -87.8452>}]
 dyn_armstrap_R_01_01 [{shapeType=Capsule, length=5, radius=2, shapeName=dyn_armstrap_L_01_01_capsule, center=<5 0 0>, rotation=<rot 90 0 0>}]
 dyn_armstrap_R_01_02 [{shapeType=Capsule, length=5, radius=1, shapeName=dyn_armstrap_L_01_02_capsule, center=<5 0 0>, rotation=<rot 90 0 0>}]
 dyn_armstrap_R_01_03 [{shapeType=Capsule, length=5, radius=1.5, shapeName=dyn_armstrap_L_01_03_capsule, center=<5 0 0>, rotation=<rot 90 0 0>}]
 dyn_armstrap_R_01_04 [{shapeType=Capsule, length=5, radius=1.5, shapeName=dyn_armstrap_L_01_03_capsule, center=<5 0 0>, rotation=<rot 90 0 0>}]
 dyn_armstrap_R_01_mstr [{shapeType=Capsule, length=0, radius=0.5, shapeName=dyn_armstrap_L_01_mstr_capsule, rotation=<rot 90 0 0>}]
 dyn_armstrap_R_02_01 [{shapeType=Capsule, length=5, radius=2, shapeName=dyn_armstrap_L_01_01_capsule, center=<5 0 0>, rotation=<rot 90 0 0>}]
 dyn_armstrap_R_02_02 [{shapeType=Capsule, length=5, radius=1, shapeName=dyn_armstrap_L_01_02_capsule, center=<5 0 0>, rotation=<rot 90 0 0>}]
 dyn_armstrap_R_02_03 [{shapeType=Capsule, length=5, radius=1.5, shapeName=dyn_armstrap_L_01_03_capsule, center=<5 0 0>, rotation=<rot 90 0 0>}]
 dyn_armstrap_R_02_04 [{shapeType=Capsule, length=5, radius=1.5, shapeName=dyn_armstrap_L_01_03_capsule, center=<5 0 0>, rotation=<rot 90 0 0>}]
 dyn_armstrap_R_02_mstr [{shapeType=Capsule, length=0, radius=0.5, shapeName=dyn_armstrap_L_01_mstr_capsule, rotation=<rot 90 0 0>}]
 dyn_armstrap_R_03_01 [{shapeType=Capsule, length=3.75, radius=2.25, shapeName=dyn_armstrap_L_01_01_capsule, center=<3 0 0>, rotation=<rot 90 0 0>}]
 dyn_armstrap_R_03_02 [{shapeType=Capsule, length=5, radius=3, shapeName=dyn_armstrap_L_01_02_capsule, center=<5 0 0>, rotation=<rot 90 0 0>}]
 dyn_armstrap_R_03_03 [{shapeType=Capsule, length=5, radius=3, shapeName=dyn_armstrap_L_01_03_capsule, center=<5 0 0>, rotation=<rot 90 0 0>}]
 dyn_armstrap_R_03_04 [{shapeType=Capsule, length=5, radius=3, shapeName=dyn_armstrap_L_01_03_capsule, center=<5 0 0>, rotation=<rot 90 0 0>}]
 dyn_armstrap_R_03_mstr [{shapeType=Capsule, length=0, radius=0.5, shapeName=dyn_armstrap_L_01_mstr_capsule, rotation=<rot 90 0 0>}]
 hand_r [{shapeType=Capsule, length=8.6852, radius=6.1455, center=<-8.2293 2.007 -0.2791>, rotation=<rot 0 76.4147 -74.5755>}]
 lowerarm_r [{shapeType=Capsule, length=18.6138, radius=5.1109, center=<-10.7783 0 0>, rotation=<rot 90 0 0>}]
 pelvis [{shapeType=Capsule, length=34.0116, radius=16.7244, rotation=<rot 90 0 0>}]
 spine_05 [{shapeType=Capsule, length=19.0864, radius=12.939, center=<1.1702 -1.2832 0.7067>, rotation=<rot 0 -135.5378 -177.0553>}]
 thigh_r [{
   shapeType=Capsule
   length=11.0602
   radius=8.7692
   shapeName=thigh_r_capsule
   center=<35.2336 -0.9798 -0.7999>
   rotation=<rot 0 -165.6411 -70.7233>
  }]
 upperarm_r [{shapeType=Capsule, length=15.8448, radius=9.1249, center=<-12.5514 1.6258 0.9539>, rotation=<rot 0 -85.1003 -86.9037>}]
}
constraints [
 {bone1Name=spine_05, bone2Name=pelvis, swing1LimitDegrees=45, swing2LimitDegrees=45, twistLimitDegrees=45}
 {
  bone1Name=thigh_r
  bone2Name=pelvis
  swing1LimitDegrees=45
  swing1Motion=Free
  swing2LimitDegrees=45
  swing2Motion=Free
  twistLimitDegrees=45
  twistMotion=Free
 }
 {
  bone1Name=calf_r
  bone2Name=thigh_r
  swing1LimitDegrees=45
  swing1Motion=Free
  swing2LimitDegrees=45
  swing2Motion=Free
  twistLimitDegrees=45
  twistMotion=Free
 }
 {
  bone1Name=upperarm_r
  bone2Name=spine_05
  swing1LimitDegrees=45
  swing1Motion=Free
  swing2LimitDegrees=45
  swing2Motion=Free
  twistLimitDegrees=45
  twistMotion=Free
 }
 {
  bone1Name=lowerarm_r
  bone2Name=upperarm_r
  swing1LimitDegrees=45
  swing1Motion=Free
  swing2LimitDegrees=45
  swing2Motion=Free
  twistLimitDegrees=45
  twistMotion=Free
 }
 {
  bone1Name=hand_r
  bone2Name=lowerarm_r
  swing1LimitDegrees=45
  swing1Motion=Free
  swing2LimitDegrees=45
  swing2Motion=Free
  twistLimitDegrees=45
  twistMotion=Free
 }
 {
  bone1Name=dyn_armstrap_R_01_mstr
  bone2Name=lowerarm_r
  swing1LimitDegrees=45
  swing1Motion=Free
  swing2LimitDegrees=45
  swing2Motion=Free
  twistLimitDegrees=45
  twistMotion=Free
 }
 {bone1Name=dyn_armstrap_R_01_01, bone2Name=dyn_armstrap_R_01_mstr, swing1LimitDegrees=60, swing2LimitDegrees=45, twistLimitDegrees=2}
 {bone1Name=dyn_armstrap_R_01_02, bone2Name=dyn_armstrap_R_01_01, swing1LimitDegrees=45, swing2LimitDegrees=45, twistLimitDegrees=2}
 {bone1Name=dyn_armstrap_R_01_03, bone2Name=dyn_armstrap_R_01_02, swing1LimitDegrees=45, swing2LimitDegrees=45, twistLimitDegrees=2}
 {
  bone1Name=dyn_armstrap_R_02_mstr
  bone2Name=lowerarm_r
  swing1LimitDegrees=45
  swing1Motion=Free
  swing2LimitDegrees=45
  swing2Motion=Free
  twistLimitDegrees=45
  twistMotion=Free
 }
 {bone1Name=dyn_armstrap_R_02_01, bone2Name=dyn_armstrap_R_02_mstr, swing1LimitDegrees=65, swing2LimitDegrees=45, twistLimitDegrees=2}
 {bone1Name=dyn_armstrap_R_02_02, bone2Name=dyn_armstrap_R_02_01, swing1LimitDegrees=45, swing2LimitDegrees=45, twistLimitDegrees=2}
 {bone1Name=dyn_armstrap_R_02_03, bone2Name=dyn_armstrap_R_02_02, swing1LimitDegrees=45, swing2LimitDegrees=45, twistLimitDegrees=2}
 {
  bone1Name=dyn_armstrap_R_03_mstr
  bone2Name=lowerarm_r
  swing1LimitDegrees=45
  swing1Motion=Free
  swing2LimitDegrees=45
  swing2Motion=Free
  twistLimitDegrees=45
  twistMotion=Free
 }
 {bone1Name=dyn_armstrap_R_03_01, bone2Name=dyn_armstrap_R_03_mstr, swing1LimitDegrees=20, swing2LimitDegrees=20, twistLimitDegrees=1}
 {bone1Name=dyn_armstrap_R_03_02, bone2Name=dyn_armstrap_R_03_01, swing1LimitDegrees=45, swing2LimitDegrees=45, twistLimitDegrees=2}
 {bone1Name=dyn_armstrap_R_03_03, bone2Name=dyn_armstrap_R_03_02, swing1LimitDegrees=45, swing2LimitDegrees=45, twistLimitDegrees=2}
 {bone1Name=dyn_armstrap_R_01_04, bone2Name=dyn_armstrap_R_01_03, swing1LimitDegrees=65, swing2LimitDegrees=65, twistLimitDegrees=2}
 {bone1Name=dyn_armstrap_R_02_04, bone2Name=dyn_armstrap_R_02_03, swing1LimitDegrees=65, swing2LimitDegrees=65, twistLimitDegrees=2}
 {bone1Name=dyn_armstrap_R_03_04, bone2Name=dyn_armstrap_R_03_03, swing1LimitDegrees=65, swing2LimitDegrees=65, twistLimitDegrees=2}
]
