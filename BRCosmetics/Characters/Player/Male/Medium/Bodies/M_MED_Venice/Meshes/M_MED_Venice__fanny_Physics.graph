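# Physics asset for the fanny mesh of the M_MED_Venice male character body cosmetic, covering its dangling tassel bones.
# Small capsules form a branching chain from pelvis through dyn_tassel_mstr to dyn_tassel_01, which forks into dyn_tassel_02 and dyn_tassel_03.
# Larger capsules on the pelvis and right thigh give body volume around the tassel.
# All constraints leave swing and twist Free, so their 45-degree limits are inactive.
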PhysicsAsset /BRCosmetics/Characters/Player/Male/Medium/Bodies/M_MED_Venice/Meshes/M_MED_Venice__fanny_Physics
bodies {
 dyn_tassel_01 [{
   shapeType=Capsule
   length=1.2937
   radius=1.1843
   shapeName=dyn_tassel_01_capsule
   center=<2.6394 0.2158 -0.0605>
   rotation=<rot -9.7373 6.004 -73.9864>
  }]
 dyn_tassel_02 [{
   shapeType=Capsule
   length=3.9563
   radius=1.2312
   shapeName=dyn_tassel_02_capsule
   center=<3.1215 0.0384 -0.2642>
   rotation=<rot 0 -90.4069 -83.2375>
  }]
 dyn_tassel_03 [{
   shapeType=Capsule
   length=2.9933
   radius=1.1235
   shapeName=dyn_tassel_03_capsule
   center=<2.3119 0.1116 0.0209>
   rotation=<rot 0 91.5745 -88.4643>
  }]
 dyn_tassel_mstr [{shapeType=Capsule, length=0.505, radius=0.505, shapeName=dyn_tassel_mstr_capsule, rotation=<rot 0 -90 0>}]
 pelvis [{shapeType=Capsule, length=9.75, radius=11, shapeName=pelvis_capsule, center=<-2.3651 0 0>}]
 thigh_r [{shapeType=Capsule, length=9.2371, radius=8.8061, shapeName=thigh_r_capsule, center=<22.2947 0 0>, rotation=<rot 90 0 0>}]
}
constraints [
 {
  bone1Name=thigh_r
  bone2Name=pelvis
  swing1LimitDegrees=45
  swing1Motion=Free
  swing2LimitDegrees=45
  swing2Motion=Free
  twistLimitDegrees=45
  twistMotion=Free
 }
 {
  bone1Name=dyn_tassel_mstr
  bone2Name=pelvis
  swing1LimitDegrees=45
  swing1Motion=Free
  swing2LimitDegrees=45
  swing2Motion=Free
  twistLimitDegrees=45
  twistMotion=Free
 }
 {
  bone1Name=dyn_tassel_01
  bone2Name=dyn_tassel_mstr
  swing1LimitDegrees=45
  swing1Motion=Free
  swing2LimitDegrees=45
  swing2Motion=Free
  twistLimitDegrees=45
  twistMotion=Free
 }
 {
  bone1Name=dyn_tassel_02
  bone2Name=dyn_tassel_01
  swing1LimitDegrees=45
  swing1Motion=Free
  swing2LimitDegrees=45
  swing2Motion=Free
  twistLimitDegrees=45
  twistMotion=Free
 }
 {
  bone1Name=dyn_tassel_03
  bone2Name=dyn_tassel_01
  swing1LimitDegrees=45
  swing1Motion=Free
  swing2LimitDegrees=45
  swing2Motion=Free
  twistLimitDegrees=45
  twistMotion=Free
 }
]
