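# Animation Blueprint for the M_MED_Venice cosmetic player body, adding secondary motion to its necklace and arm straps.
# It copies the attached parent mesh's pose, then runs deform and control Control Rigs.
# Weighted constraints place necklace and strap master bones before a long series of per-bone AnimDynamics nodes simulates them.
# Straps use gravity overrides and mostly spherical limits on pelvis and forearms as cheap body collision.
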AnimBlueprint /BRCosmetics/Characters/Player/Male/Medium/Bodies/M_MED_Venice/Meshes/M_MED_Venice_AnimBP
target_skeleton=&/M_MED_Venice_Skeleton
graphs [
 AnimGraph {nodes [
   Anim Dynamics - Bone: dyn_armstrap_L_01_06(Alpha=1, ExternalForce="0,0,0") {
    angularDampingOverride=0.999
    bOverrideAngularDamping=true
    bUseGravityOverride=true
    bUsePlanarLimit=false
    bUseSphericalLimits=true
    simulationSpace=RootRelative
    alphaBoolBlend {blendOption=Linear}
    boundBone {boneName=dyn_armstrap_L_01_06}
    componentAppliedLinearAccClamp=<3000 3000 3000>
    componentLinearAccScale=<5 5 8>
    componentLinearVelScale=<5 5 8>
    gravityOverride {z=-980}
    physicsBodyDefinitions [{
      boundBone {boneName=dyn_armstrap_L_01_06}
      boxExtents=<20 20 20>
      constraintSetup {angularLimitsMax=<3 55 55>, angularLimitsMin=<-3 -55 -55>, angularTarget=<0 0 0>}
      localJointOffset=<10 0 0>
     }]
    relativeSpaceBone {boneName=dyn_armstrap_L_01_03}
    sphericalLimits [{limitRadius=20, drivingBone {boneName=pelvis}, sphereLocalOffset=<-5 0 0>}]
   }
   Anim Dynamics - Bone: dyn_armstrap_L_02_06(Alpha=1, ExternalForce="0,0,0") {
    angularDampingOverride=0.999
    bOverrideAngularDamping=true
    bUseGravityOverride=true
    bUsePlanarLimit=false
    bUseSphericalLimits=true
    simulationSpace=RootRelative
    alphaBoolBlend {blendOption=Linear}
    boundBone {boneName=dyn_armstrap_L_02_06}
    componentAppliedLinearAccClamp=<3000 3000 3000>
    componentLinearAccScale=<5 5 8>
    componentLinearVelScale=<5 5 8>
    gravityOverride {z=-980}
    physicsBodyDefinitions [{
      boundBone {boneName=dyn_armstrap_L_02_06}
      boxExtents=<20 20 20>
      constraintSetup {angularLimitsMax=<3 55 55>, angularLimitsMin=<-3 -55 -55>, angularTarget=<0 0 0>}
      localJointOffset=<10 0 0>
     }]
    relativeSpaceBone {boneName=dyn_armstrap_L_01_03}
    sphericalLimits [{limitRadius=20, drivingBone {boneName=pelvis}, sphereLocalOffset=<-5 0 0>}]
   }
   "Anim Dynamics - Chain: dyn_lace_l_01_01 -> dyn_lace_l_01_03"(Alpha=0.6, ExternalForce="0,0,0") {
    angularDampingOverride=0.9
    bChain=true
    bOverrideAngularDamping=true
    simulationSpace=RootRelative
    alphaBoolBlend {blendOption=Linear}
    boundBone {boneName=dyn_lace_l_01_01}
    chainEnd {boneName=dyn_lace_l_01_03}
    componentAppliedLinearAccClamp=<10000 10000 10000>
    componentLinearAccScale=<0.1 0.1 0.3>
    componentLinearVelScale=<0.1 0.1 0.3>
    physicsBodyDefinitions [
     {
      boundBone {boneName=dyn_lace_l_01_01}
      boxExtents=<50 50 50>
      constraintSetup {angularLimitsMax=<0 2 25>, angularLimitsMin=<0 -20 -25>, angularTarget=<0 0 0>}
      localJointOffset=<0.5 0 0>
     }
     {
      boundBone {boneName=dyn_lace_l_01_02}
      boxExtents=<30 30 30>
      constraintSetup {angularLimitsMax=<0 2 25>, angularLimitsMin=<0 -20 -25>, angularTarget=<0 0 0>}
      localJointOffset=<0.5 0 0>
     }
     {
      boundBone {boneName=dyn_lace_l_01_03}
      boxExtents=<20 20 20>
      constraintSetup {angularLimitsMax=<0 2 25>, angularLimitsMin=<0 -20 -25>, angularTarget=<0 0 0>}
      localJointOffset=<0.5 0 0>
     }
    ]
    relativeSpaceBone {boneName=foot_l}
   }
   "Anim Dynamics - Chain: dyn_lace_l_02_01 -> dyn_lace_l_02_03"(Alpha=0.6, ExternalForce="0,0,0") {
    angularDampingOverride=0.9
    angularSpringConstant=0.9999
    bAngularSpring=true
    bChain=true
    bOverrideAngularDamping=true
    simulationSpace=RootRelative
    alphaBoolBlend {blendOption=Linear}
    boundBone {boneName=dyn_lace_l_02_01}
    chainEnd {boneName=dyn_lace_l_02_03}
    componentAppliedLinearAccClamp=<10000 10000 10000>
    componentLinearAccScale=<0.1 0.1 0.3>
    componentLinearVelScale=<0.1 0.1 0.3>
    physicsBodyDefinitions [
     {
      boundBone {boneName=dyn_lace_l_02_01}
      boxExtents=<50 50 50>
      constraintSetup {angularLimitsMax=<0 2 25>, angularLimitsMin=<0 -20 -25>, angularTarget=<0 0 0>}
      localJointOffset=<0.5 0 0>
     }
     {
      boundBone {boneName=dyn_lace_l_02_02}
      boxExtents=<30 30 30>
      constraintSetup {angularLimitsMax=<0 2 25>, angularLimitsMin=<0 -20 -25>, angularTarget=<0 0 0>}
      localJointOffset=<0.5 0 0>
     }
     {
      boundBone {boneName=dyn_lace_l_02_03}
      boxExtents=<20 20 20>
      constraintSetup {angularLimitsMax=<0 2 25>, angularLimitsMin=<0 -20 -25>, angularTarget=<0 0 0>}
      localJointOffset=<0.5 0 0>
     }
    ]
    relativeSpaceBone {boneName=foot_l}
   }
   "Anim Dynamics - Chain: dyn_lace_r_01_01 -> dyn_lace_r_01_03"(Alpha=0.6, ExternalForce="0,0,0") {
    angularDampingOverride=0.9
    angularSpringConstant=0.99
    bAngularSpring=true
    bChain=true
    bOverrideAngularDamping=true
    simulationSpace=RootRelative
    alphaBoolBlend {blendOption=Linear}
    boundBone {boneName=dyn_lace_r_01_01}
    chainEnd {boneName=dyn_lace_r_01_03}
    componentAppliedLinearAccClamp=<10000 10000 10000>
    componentLinearAccScale=<0.1 0.1 0.3>
    componentLinearVelScale=<0.1 0.1 0.3>
    physicsBodyDefinitions [
     {
      boundBone {boneName=dyn_lace_r_01_01}
      boxExtents=<50 50 50>
      constraintSetup {angularLimitsMax=<0 2 25>, angularLimitsMin=<0 -20 -25>, angularTarget=<0 0 0>}
      localJointOffset=<0.5 0 0>
     }
     {
      boundBone {boneName=dyn_lace_r_01_02}
      boxExtents=<30 30 30>
      constraintSetup {angularLimitsMax=<0 2 25>, angularLimitsMin=<0 -20 -25>, angularTarget=<0 0 0>}
      localJointOffset=<0.5 0 0>
     }
     {
      boundBone {boneName=dyn_lace_r_01_03}
      boxExtents=<20 20 20>
      constraintSetup {angularLimitsMax=<0 2 25>, angularLimitsMin=<0 -20 -25>, angularTarget=<0 0 0>}
      localJointOffset=<0.5 0 0>
     }
    ]
    relativeSpaceBone {boneName=foot_l}
   }
   "Anim Dynamics - Chain: dyn_lace_r_02_01 -> dyn_lace_r_02_03"(Alpha=0.7, ExternalForce="0,0,0") {
    angularDampingOverride=0.9
    angularSpringConstant=0.9999
    bChain=true
    bOverrideAngularDamping=true
    bUseGravityOverride=true
    simulationSpace=RootRelative
    alphaBoolBlend {blendOption=Linear}
    boundBone {boneName=dyn_lace_r_02_01}
    chainEnd {boneName=dyn_lace_r_02_03}
    componentAppliedLinearAccClamp=<1000 1000 1000>
    componentLinearAccScale=<0.1 0.1 0.3>
    componentLinearVelScale=<0.1 0.1 0.3>
    gravityOverride {z=-980}
    physicsBodyDefinitions [
     {
      boundBone {boneName=dyn_lace_r_02_01}
      boxExtents=<50 50 50>
      constraintSetup {angularLimitsMax=<0 0 90>, angularLimitsMin=<0 0 -90>, angularTarget=<0 0 0>}
      localJointOffset=<0.1 0 0>
     }
     {
      boundBone {boneName=dyn_lace_r_02_02}
      boxExtents=<30 30 30>
      constraintSetup {angularLimitsMax=<0 0 90>, angularLimitsMin=<0 0 -90>, angularTarget=<0 0 0>}
      localJointOffset=<0.1 0 0>
     }
     {
      boundBone {boneName=dyn_lace_r_02_03}
      boxExtents=<20 20 20>
      constraintSetup {angularLimitsMax=<0 0 90>, angularLimitsMin=<0 0 -90>, angularTarget=<0 0 0>}
      localJointOffset=<0.1 0 0>
     }
    ]
    planarLimits [{drivingBone {boneName=foot_r}, planeTransform {location=<10 0 0>, rotation=<rot 20 180 180>}}]
    relativeSpaceBone {boneName=foot_l}
   }
   "Transform (Modify) Bone - Bone: dyn_necklace_mstr"(Alpha=1, Translation="0,200,200", Rotation="0,0,0", Scale="1,1,1") {translationMode=BMM_Additive, alphaBoolBlend {blendOption=Linear}, boneToModify {boneName=dyn_necklace_mstr}, translation {y=200, z=200}}
   n1 = Get GravityOverride()
   RigidBody!(ComponentPose, Alpha, OverrideWorldGravity:Vector=@n1, ExternalForce) -> (Pose) {
    bOverrideWorldGravity=true
    overridePhysicsAsset=&/M_MED_Venice_Physics_L
    simulationSpace=BaseBoneSpace
    alphaBoolBlend {blendOption=Linear}
    baseBoneRef {boneName=upperarm_l}
    componentAppliedLinearAccClamp=<1000 1000 1000>
    componentLinearAccScale=<0.2 0.2 1>
    componentLinearVelScale=<0.2 0.2 1>
    overrideWorldGravity {z=-980}
    simSpaceSettings {maxAngularAcceleration=1000, maxAngularVelocity=1000, maxLinearAcceleration=1000, maxLinearVelocity=1000, velocityScaleZ=0.8, worldAlpha=0.8}
   }
   RigidBody!(ComponentPose, Alpha, ExternalForce) -> (Pose) {
    overridePhysicsAsset=&/M_MED_Venice__fanny_Physics
    simulationSpace=BaseBoneSpace
    alphaBoolBlend {blendOption=Linear}
    baseBoneRef {boneName=pelvis}
    componentLinearAccScale=<0.3 0.3 0.5>
    componentLinearVelScale=<0.3 0.3 0.5>
   }
   RigidBody!(ComponentPose, Alpha, OverrideWorldGravity:Vector=@n1, ExternalForce) -> (Pose) {
    bOverrideWorldGravity=true
    overridePhysicsAsset=&/M_MED_Venice_Physics_R
    simulationSpace=BaseBoneSpace
    alphaBoolBlend {blendOption=Linear}
    baseBoneRef {boneName=upperarm_r}
    componentAppliedLinearAccClamp=<1000 1000 1000>
    componentLinearAccScale=<0.2 0.2 1>
    componentLinearVelScale=<0.2 0.2 1>
    overrideWorldGravity {z=-980}
    simSpaceSettings {maxAngularAcceleration=1000, maxAngularVelocity=1000, maxLinearAcceleration=1000, maxLinearVelocity=1000, velocityScaleZ=0.8, worldAlpha=0.8}
   }
   n2 = Control Rig(Source=Copy Pose From Mesh() {bUseAttachedParent=true}) {
    alphaBoolBlend {blendOption=Linear}
    controlRigAssetReference {blueprintRigClass=&/M_Venice_DeformRig.M_Venice_DeformRig_C}
   }
   n3 = Control Rig(Source=@n2) {
    alphaBoolBlend {blendOption=Linear}
    controlRigAssetReference {blueprintRigClass=&/M_MED_Venice_CtrlRig.M_MED_Venice_CtrlRig_C}
   }
   n4 = "Constraint - dyn_necklace_mstr (PPP)"!(
    ComponentPose=Local To Component(LocalPose=@n3)
    Alpha=1
    ConstraintWeights_0=0.8
    ConstraintWeights_1:float=0.15
    ConstraintWeights_2:float=0.05
   ) -> (Pose) {
    alphaBoolBlend {blendOption=Linear}
    boneToModify {boneName=dyn_necklace_mstr}
    constraintSetup [{targetBone {boneName=spine_05}}, {targetBone {boneName=spine_04}}, {targetBone {boneName=neck_01}}]
    constraintWeights [1,1,1]
   }
   n5 = Anim Dynamics - Bone: dyn_necklace_01(ComponentPose=@n4, Alpha=1, ExternalForce="0,0,0") {
    angularDampingOverride=0.9999
    bOverrideAngularDamping=true
    alphaBoolBlend {blendOption=Linear}
    boundBone {boneName=dyn_necklace_01}
    componentAppliedLinearAccClamp=<10000 10000 10000>
    componentLinearAccScale=<0.3 0.3 0.5>
    componentLinearVelScale=<0.3 0.3 0.5>
    physicsBodyDefinitions [{
      boundBone {boneName=dyn_necklace_01}
      boxExtents=<30 30 30>
      constraintSetup {angularLimitsMax=<0 10 15>, angularLimitsMin=<0 -10 -5>, angularTarget=<0 0 0>}
      localJointOffset=<1 0 0>
     }]
   }
   n6 = Anim Dynamics - Bone: dyn_necklace_02(ComponentPose=@n5, Alpha=1, ExternalForce="0,0,0") {
    angularDampingOverride=0.9999
    bOverrideAngularDamping=true
    alphaBoolBlend {blendOption=Linear}
    boundBone {boneName=dyn_necklace_02}
    componentAppliedLinearAccClamp=<10000 10000 10000>
    componentLinearAccScale=<0.3 0.3 0.5>
    componentLinearVelScale=<0.3 0.3 0.5>
    physicsBodyDefinitions [{
      boundBone {boneName=dyn_necklace_02}
      boxExtents=<30 30 30>
      constraintSetup {angularLimitsMax=<0 10 15>, angularLimitsMin=<0 -10 -2>, angularTarget=<0 0 0>}
      localJointOffset=<1 0 0>
     }]
   }
   n7 = Anim Dynamics - Bone: dyn_necklace_03(ComponentPose=@n6, Alpha=1, ExternalForce="0,0,0") {
    angularDampingOverride=0.99
    bOverrideAngularDamping=true
    alphaBoolBlend {blendOption=Linear}
    boundBone {boneName=dyn_necklace_03}
    componentAppliedLinearAccClamp=<10000 10000 10000>
    componentLinearAccScale=<0.3 0.3 0.5>
    componentLinearVelScale=<0.3 0.3 0.5>
    physicsBodyDefinitions [{
      boundBone {boneName=dyn_necklace_03}
      boxExtents=<15 15 15>
      constraintSetup {angularLimitsMax=<0 15 30>, angularLimitsMin=<0 -15 0>, angularTarget=<0 0 0>}
      localJointOffset=<1 0 0>
     }]
   }
   n8 = "Constraint - dyn_armstrap_L_03_mstr (PPPPP)"!(
    ComponentPose=@n7
    Alpha=1
    ConstraintWeights_0=0
    ConstraintWeights_1:float=0.64
    ConstraintWeights_2:float=0.1
    ConstraintWeights_3:float=0.16
    ConstraintWeights_4:float=0.1
   ) -> (Pose) {
    alphaBoolBlend {blendOption=Linear}
    boneToModify {boneName=dyn_armstrap_L_03_mstr}
    constraintSetup [
     {targetBone {boneName=upperarm_twist_01_l}}
     {targetBone {boneName=upperarm_l}}
     {targetBone {boneName=lowerarm_twist_02_l}}
     {targetBone {boneName=upperarm_twist_02_l}}
     {targetBone {boneName=lowerarm_l}}
    ]
    constraintWeights [1,1,1,1,1]
   }
   n9 = "Constraint - dyn_armstrap_L_02_mstr (PPPP)"!(
    ComponentPose=@n8
    Alpha=1
    ConstraintWeights_0=0.27
    ConstraintWeights_1:float=0.43
    ConstraintWeights_2:float=0.03
    ConstraintWeights_3:float=0.27
   ) -> (Pose) {
    alphaBoolBlend {blendOption=Linear}
    boneToModify {boneName=dyn_armstrap_L_02_mstr}
    constraintSetup [
     {targetBone {boneName=upperarm_l}}
     {targetBone {boneName=lowerarm_twist_02_l}}
     {targetBone {boneName=upperarm_twist_02_l}}
     {targetBone {boneName=lowerarm_l}}
    ]
    constraintWeights [1,1,1,1]
   }
   n10 = "Constraint - dyn_armstrap_L_01_mstr (PPPP)"!(
    ComponentPose=@n9
    Alpha=1
    ConstraintWeights_0=0.27
    ConstraintWeights_1:float=0.43
    ConstraintWeights_2:float=0.03
    ConstraintWeights_3:float=0.27
   ) -> (Pose) {
    alphaBoolBlend {blendOption=Linear}
    boneToModify {boneName=dyn_armstrap_L_01_mstr}
    constraintSetup [
     {targetBone {boneName=upperarm_l}}
     {targetBone {boneName=lowerarm_twist_02_l}}
     {targetBone {boneName=upperarm_twist_02_l}}
     {targetBone {boneName=lowerarm_l}}
    ]
    constraintWeights [1,1,1,1]
   }
   n11 = "Constraint - dyn_armstrap_R_03_mstr (PPPP)"!(
    ComponentPose=@n10
    Alpha=1
    ConstraintWeights_0=0.18
    ConstraintWeights_1:float=0.55
    ConstraintWeights_2:float=0.09
    ConstraintWeights_3:float=0.18
   ) -> (Pose) {
    alphaBoolBlend {blendOption=Linear}
    boneToModify {boneName=dyn_armstrap_R_03_mstr}
    constraintSetup [
     {targetBone {boneName=lowerarm_twist_02_r}}
     {targetBone {boneName=upperarm_r}}
     {targetBone {boneName=upperarm_twist_02_r}}
     {targetBone {boneName=lowerarm_r}}
    ]
    constraintWeights [1,1,1,1]
   }
   n12 = "Constraint - dyn_armstrap_R_01_mstr (PPPP)"!(
    ComponentPose=@n11
    Alpha=1
    ConstraintWeights_0=0.58
    ConstraintWeights_1:float=0.15
    ConstraintWeights_2:float=0.05
    ConstraintWeights_3:float=0.22
   ) -> (Pose) {
    alphaBoolBlend {blendOption=Linear}
    boneToModify {boneName=dyn_armstrap_R_01_mstr}
    constraintSetup [
     {targetBone {boneName=lowerarm_twist_02_r}}
     {targetBone {boneName=upperarm_r}}
     {targetBone {boneName=lowerarm_twist_01_r}}
     {targetBone {boneName=lowerarm_r}}
    ]
    constraintWeights [1,1,1,1]
   }
   n13 = "Constraint - dyn_armstrap_R_02_mstr (PPPP)"!(
    ComponentPose=@n12
    Alpha=1
    ConstraintWeights_0=0.58
    ConstraintWeights_1:float=0.15
    ConstraintWeights_2:float=0.05
    ConstraintWeights_3:float=0.22
   ) -> (Pose) {
    alphaBoolBlend {blendOption=Linear}
    boneToModify {boneName=dyn_armstrap_R_02_mstr}
    constraintSetup [
     {targetBone {boneName=lowerarm_twist_02_r}}
     {targetBone {boneName=upperarm_r}}
     {targetBone {boneName=lowerarm_twist_01_r}}
     {targetBone {boneName=lowerarm_r}}
    ]
    constraintWeights [1,1,1,1]
   }
   n14 = Anim Dynamics - Bone: dyn_armstrap_L_01_01(ComponentPose=@n13, Alpha=1, ExternalForce="0,0,0") {
    angularDampingOverride=1
    bOverrideAngularDamping=true
    bUseGravityOverride=true
    bUsePlanarLimit=false
    simulationSpace=RootRelative
    alphaBoolBlend {blendOption=Linear}
    boundBone {boneName=dyn_armstrap_L_01_01}
    componentAppliedLinearAccClamp=<3000 3000 3000>
    componentLinearAccScale=<5 5 8>
    componentLinearVelScale=<5 5 8>
    gravityOverride {z=-980}
    physicsBodyDefinitions [{
      boundBone {boneName=dyn_armstrap_L_01_01}
      boxExtents=<25 25 25>
      constraintSetup {angularLimitsMax=<3 25 5>, angularLimitsMin=<-3 -25 -30>, angularTarget=<0 0 0>}
      localJointOffset=<10 0 0>
     }]
    relativeSpaceBone {boneName=spine_05}
    sphericalLimits [{limitRadius=20, drivingBone {boneName=pelvis}, sphereLocalOffset=<-5 0 0>}]
   }
   n15 = Anim Dynamics - Bone: dyn_armstrap_L_01_02(ComponentPose=@n14, Alpha=1, ExternalForce="0,0,0") {
    angularDampingOverride=0.9999
    bOverrideAngularDamping=true
    bUseGravityOverride=true
    bUsePlanarLimit=false
    bUseSphericalLimits=true
    simulationSpace=RootRelative
    alphaBoolBlend {blendOption=Linear}
    boundBone {boneName=dyn_armstrap_L_01_02}
    componentAppliedLinearAccClamp=<3000 3000 3000>
    componentLinearAccScale=<5 5 8>
    componentLinearVelScale=<5 5 8>
    gravityOverride {z=-980}
    physicsBodyDefinitions [{
      boundBone {boneName=dyn_armstrap_L_01_02}
      boxExtents=<25 25 25>
      constraintSetup {angularLimitsMax=<3 45 45>, angularLimitsMin=<-3 -45 -45>, angularTarget=<0 0 0>}
      localJointOffset=<10 0 0>
     }]
    relativeSpaceBone {boneName=dyn_armstrap_L_01_01}
    sphericalLimits [{limitRadius=20, drivingBone {boneName=pelvis}, sphereLocalOffset=<-5 0 0>}]
   }
   n16 = Anim Dynamics - Bone: dyn_armstrap_L_01_03(ComponentPose=@n15, Alpha=1, ExternalForce="0,0,0") {
    angularDampingOverride=0.999
    bOverrideAngularDamping=true
    bUseGravityOverride=true
    bUsePlanarLimit=false
    bUseSphericalLimits=true
    simulationSpace=RootRelative
    alphaBoolBlend {blendOption=Linear}
    boundBone {boneName=dyn_armstrap_L_01_03}
    componentAppliedLinearAccClamp=<3000 3000 3000>
    componentLinearAccScale=<5 5 8>
    componentLinearVelScale=<5 5 8>
    gravityOverride {z=-980}
    physicsBodyDefinitions [{
      boundBone {boneName=dyn_armstrap_L_01_03}
      boxExtents=<25 25 25>
      constraintSetup {angularLimitsMax=<3 55 55>, angularLimitsMin=<-3 -55 -55>, angularTarget=<0 0 0>}
      localJointOffset=<10 0 0>
     }]
    relativeSpaceBone {boneName=dyn_armstrap_L_01_02}
    sphericalLimits [
     {limitRadius=20, drivingBone {boneName=pelvis}, sphereLocalOffset=<-5 0 0>}
     {limitRadius=13, drivingBone {boneName=lowerarm_l}, sphereLocalOffset=<20 0 0>}
    ]
   }
   n17 = Anim Dynamics - Bone: dyn_armstrap_L_01_04(ComponentPose=@n16, Alpha=1, ExternalForce="0,0,0") {
    angularDampingOverride=0.999
    bOverrideAngularDamping=true
    bUseGravityOverride=true
    bUsePlanarLimit=false
    bUseSphericalLimits=true
    simulationSpace=RootRelative
    alphaBoolBlend {blendOption=Linear}
    boundBone {boneName=dyn_armstrap_L_01_04}
    componentAppliedLinearAccClamp=<3000 3000 3000>
    componentLinearAccScale=<5 5 8>
    componentLinearVelScale=<5 5 8>
    gravityOverride {z=-980}
    physicsBodyDefinitions [{
      boundBone {boneName=dyn_armstrap_L_01_04}
      boxExtents=<20 20 20>
      constraintSetup {angularLimitsMax=<3 55 55>, angularLimitsMin=<-3 -55 -55>, angularTarget=<0 0 0>}
      localJointOffset=<10 0 0>
     }]
    relativeSpaceBone {boneName=dyn_armstrap_L_01_03}
    sphericalLimits [
     {limitRadius=20, drivingBone {boneName=pelvis}, sphereLocalOffset=<-5 0 0>}
     {limitRadius=13, drivingBone {boneName=lowerarm_l}, sphereLocalOffset=<30 0 0>}
    ]
   }
   n18 = Anim Dynamics - Bone: dyn_armstrap_L_01_05(ComponentPose=@n17, Alpha=1, ExternalForce="0,0,0") {
    angularDampingOverride=0.999
    bOverrideAngularDamping=true
    bUseGravityOverride=true
    bUsePlanarLimit=false
    bUseSphericalLimits=true
    simulationSpace=RootRelative
    alphaBoolBlend {blendOption=Linear}
    boundBone {boneName=dyn_armstrap_L_01_05}
    componentAppliedLinearAccClamp=<3000 3000 3000>
    componentLinearAccScale=<5 5 8>
    componentLinearVelScale=<5 5 8>
    gravityOverride {z=-980}
    physicsBodyDefinitions [{
      boundBone {boneName=dyn_armstrap_L_01_05}
      boxExtents=<20 20 20>
      constraintSetup {angularLimitsMax=<3 55 55>, angularLimitsMin=<-3 -55 -55>, angularTarget=<0 0 0>}
      localJointOffset=<10 0 0>
     }]
    relativeSpaceBone {boneName=dyn_armstrap_L_01_03}
    sphericalLimits [
     {limitRadius=20, drivingBone {boneName=pelvis}, sphereLocalOffset=<-5 0 0>}
     {limitRadius=15, drivingBone {boneName=lowerarm_l}, sphereLocalOffset=<30 0 0>}
    ]
   }
   n19 = Anim Dynamics - Bone: dyn_armstrap_L_02_01(ComponentPose=@n18, Alpha=1, ExternalForce="0,0,0") {
    angularDampingOverride=1
    bOverrideAngularDamping=true
    bUseGravityOverride=true
    bUsePlanarLimit=false
    simulationSpace=RootRelative
    alphaBoolBlend {blendOption=Linear}
    boundBone {boneName=dyn_armstrap_L_02_01}
    componentAppliedLinearAccClamp=<3000 3000 3000>
    componentLinearAccScale=<5 5 8>
    componentLinearVelScale=<5 5 8>
    gravityOverride {z=-980}
    physicsBodyDefinitions [{
      boundBone {boneName=dyn_armstrap_L_02_01}
      boxExtents=<25 25 25>
      constraintSetup {angularLimitsMax=<3 25 5>, angularLimitsMin=<-3 -25 -30>, angularTarget=<0 0 0>}
      localJointOffset=<10 0 0>
     }]
    relativeSpaceBone {boneName=spine_05}
    sphericalLimits [{limitRadius=20, drivingBone {boneName=pelvis}, sphereLocalOffset=<-5 0 0>}]
   }
   n20 = Anim Dynamics - Bone: dyn_armstrap_L_02_02(ComponentPose=@n19, Alpha=1, ExternalForce="0,0,0") {
    angularDampingOverride=0.9999
    bOverrideAngularDamping=true
    bUseGravityOverride=true
    bUsePlanarLimit=false
    bUseSphericalLimits=true
    simulationSpace=RootRelative
    alphaBoolBlend {blendOption=Linear}
    boundBone {boneName=dyn_armstrap_L_02_02}
    componentAppliedLinearAccClamp=<3000 3000 3000>
    componentLinearAccScale=<5 5 8>
    componentLinearVelScale=<5 5 8>
    gravityOverride {z=-980}
    physicsBodyDefinitions [{
      boundBone {boneName=dyn_armstrap_L_02_02}
      boxExtents=<25 25 25>
      constraintSetup {angularLimitsMax=<3 45 45>, angularLimitsMin=<-3 -45 -45>, angularTarget=<0 0 0>}
      localJointOffset=<10 0 0>
     }]
    relativeSpaceBone {boneName=dyn_armstrap_L_01_01}
    sphericalLimits [{limitRadius=20, drivingBone {boneName=pelvis}, sphereLocalOffset=<-5 0 0>}]
   }
   n21 = Anim Dynamics - Bone: dyn_armstrap_L_02_03(ComponentPose=@n20, Alpha=1, ExternalForce="0,0,0") {
    angularDampingOverride=0.999
    bOverrideAngularDamping=true
    bUseGravityOverride=true
    bUsePlanarLimit=false
    bUseSphericalLimits=true
    simulationSpace=RootRelative
    alphaBoolBlend {blendOption=Linear}
    boundBone {boneName=dyn_armstrap_L_02_03}
    componentAppliedLinearAccClamp=<3000 3000 3000>
    componentLinearAccScale=<5 5 8>
    componentLinearVelScale=<5 5 8>
    gravityOverride {z=-980}
    physicsBodyDefinitions [{
      boundBone {boneName=dyn_armstrap_L_02_03}
      boxExtents=<25 25 25>
      constraintSetup {angularLimitsMax=<3 55 55>, angularLimitsMin=<-3 -55 -55>, angularTarget=<0 0 0>}
      localJointOffset=<10 0 0>
     }]
    relativeSpaceBone {boneName=dyn_armstrap_L_01_02}
    sphericalLimits [
     {limitRadius=20, drivingBone {boneName=pelvis}, sphereLocalOffset=<-5 0 0>}
     {limitRadius=13, drivingBone {boneName=lowerarm_l}, sphereLocalOffset=<20 0 0>}
    ]
   }
   n22 = Anim Dynamics - Bone: dyn_armstrap_L_02_04(ComponentPose=@n21, Alpha=1, ExternalForce="0,0,0") {
    angularDampingOverride=0.999
    bOverrideAngularDamping=true
    bUseGravityOverride=true
    bUsePlanarLimit=false
    bUseSphericalLimits=true
    simulationSpace=RootRelative
    alphaBoolBlend {blendOption=Linear}
    boundBone {boneName=dyn_armstrap_L_02_04}
    componentAppliedLinearAccClamp=<3000 3000 3000>
    componentLinearAccScale=<5 5 8>
    componentLinearVelScale=<5 5 8>
    gravityOverride {z=-980}
    physicsBodyDefinitions [{
      boundBone {boneName=dyn_armstrap_L_02_04}
      boxExtents=<20 20 20>
      constraintSetup {angularLimitsMax=<3 55 55>, angularLimitsMin=<-3 -55 -55>, angularTarget=<0 0 0>}
      localJointOffset=<10 0 0>
     }]
    relativeSpaceBone {boneName=dyn_armstrap_L_01_03}
    sphericalLimits [
     {limitRadius=20, drivingBone {boneName=pelvis}, sphereLocalOffset=<-5 0 0>}
     {limitRadius=13, drivingBone {boneName=lowerarm_l}, sphereLocalOffset=<30 0 0>}
    ]
   }
   n23 = Anim Dynamics - Bone: dyn_armstrap_L_02_05(ComponentPose=@n22, Alpha=1, ExternalForce="0,0,0") {
    angularDampingOverride=0.999
    bOverrideAngularDamping=true
    bUseGravityOverride=true
    bUsePlanarLimit=false
    bUseSphericalLimits=true
    simulationSpace=RootRelative
    alphaBoolBlend {blendOption=Linear}
    boundBone {boneName=dyn_armstrap_L_02_05}
    componentAppliedLinearAccClamp=<3000 3000 3000>
    componentLinearAccScale=<5 5 8>
    componentLinearVelScale=<5 5 8>
    gravityOverride {z=-980}
    physicsBodyDefinitions [{
      boundBone {boneName=dyn_armstrap_L_02_05}
      boxExtents=<20 20 20>
      constraintSetup {angularLimitsMax=<3 55 55>, angularLimitsMin=<-3 -55 -55>, angularTarget=<0 0 0>}
      localJointOffset=<10 0 0>
     }]
    relativeSpaceBone {boneName=dyn_armstrap_L_01_03}
    sphericalLimits [
     {limitRadius=20, drivingBone {boneName=pelvis}, sphereLocalOffset=<-5 0 0>}
     {limitRadius=15, drivingBone {boneName=lowerarm_l}, sphereLocalOffset=<30 0 0>}
    ]
   }
   n24 = Anim Dynamics - Bone: dyn_armstrap_L_03_01(ComponentPose=@n23, Alpha=1, ExternalForce="0,0,0") {
    angularDampingOverride=1
    bOverrideAngularDamping=true
    bUseGravityOverride=true
    bUsePlanarLimit=false
    simulationSpace=RootRelative
    alphaBoolBlend {blendOption=Linear}
    boundBone {boneName=dyn_armstrap_L_03_01}
    componentAppliedLinearAccClamp=<3000 3000 3000>
    componentLinearAccScale=<5 5 8>
    componentLinearVelScale=<5 5 8>
    gravityOverride {z=-980}
    physicsBodyDefinitions [{
      boundBone {boneName=dyn_armstrap_L_03_01}
      boxExtents=<40 40 40>
      constraintSetup {angularLimitsMax=<3 25 5>, angularLimitsMin=<-3 -25 -30>, angularTarget=<0 0 0>}
      localJointOffset=<10 0 0>
     }]
    relativeSpaceBone {boneName=spine_05}
    sphericalLimits [{limitRadius=20, drivingBone {boneName=pelvis}, sphereLocalOffset=<-5 0 0>}]
   }
   n25 = Anim Dynamics - Bone: dyn_armstrap_L_03_02(ComponentPose=@n24, Alpha=1, ExternalForce="0,0,0") {
    angularDampingOverride=0.9999
    bOverrideAngularDamping=true
    bUseGravityOverride=true
    bUsePlanarLimit=false
    bUseSphericalLimits=true
    simulationSpace=RootRelative
    alphaBoolBlend {blendOption=Linear}
    boundBone {boneName=dyn_armstrap_L_03_02}
    componentAppliedLinearAccClamp=<3000 3000 3000>
    componentLinearAccScale=<5 5 8>
    componentLinearVelScale=<5 5 8>
    gravityOverride {z=-980}
    physicsBodyDefinitions [{
      boundBone {boneName=dyn_armstrap_L_03_02}
      boxExtents=<35 35 35>
      constraintSetup {angularLimitsMax=<3 45 45>, angularLimitsMin=<-3 -45 -45>, angularTarget=<0 0 0>}
      localJointOffset=<10 0 0>
     }]
    relativeSpaceBone {boneName=dyn_armstrap_L_01_01}
    sphericalLimits [{limitRadius=20, drivingBone {boneName=pelvis}, sphereLocalOffset=<-5 0 0>}]
   }
   n26 = Anim Dynamics - Bone: dyn_armstrap_L_03_03(ComponentPose=@n25, Alpha=1, ExternalForce="0,0,0") {
    angularDampingOverride=0.999
    bOverrideAngularDamping=true
    bUseGravityOverride=true
    bUsePlanarLimit=false
    bUseSphericalLimits=true
    simulationSpace=RootRelative
    alphaBoolBlend {blendOption=Linear}
    boundBone {boneName=dyn_armstrap_L_03_03}
    componentAppliedLinearAccClamp=<3000 3000 3000>
    componentLinearAccScale=<5 5 8>
    componentLinearVelScale=<5 5 8>
    gravityOverride {z=-980}
    physicsBodyDefinitions [{
      boundBone {boneName=dyn_armstrap_L_03_03}
      boxExtents=<25 25 25>
      constraintSetup {angularLimitsMax=<3 55 55>, angularLimitsMin=<-3 -55 -55>, angularTarget=<0 0 0>}
      localJointOffset=<10 0 0>
     }]
    relativeSpaceBone {boneName=dyn_armstrap_L_01_02}
    sphericalLimits [
     {limitRadius=20, drivingBone {boneName=pelvis}, sphereLocalOffset=<-5 0 0>}
     {limitRadius=13, drivingBone {boneName=lowerarm_l}, sphereLocalOffset=<20 0 0>}
    ]
   }
   n27 = Anim Dynamics - Bone: dyn_armstrap_L_03_04(ComponentPose=@n26, Alpha=1, ExternalForce="0,0,0") {
    angularDampingOverride=0.999
    bOverrideAngularDamping=true
    bUseGravityOverride=true
    bUsePlanarLimit=false
    bUseSphericalLimits=true
    simulationSpace=RootRelative
    alphaBoolBlend {blendOption=Linear}
    boundBone {boneName=dyn_armstrap_L_03_04}
    componentAppliedLinearAccClamp=<3000 3000 3000>
    componentLinearAccScale=<5 5 8>
    componentLinearVelScale=<5 5 8>
    gravityOverride {z=-980}
    physicsBodyDefinitions [{
      boundBone {boneName=dyn_armstrap_L_03_04}
      boxExtents=<15 15 15>
      constraintSetup {angularLimitsMax=<3 55 55>, angularLimitsMin=<-3 -55 -55>, angularTarget=<0 0 0>}
      localJointOffset=<10 0 0>
     }]
    relativeSpaceBone {boneName=dyn_armstrap_L_01_03}
    sphericalLimits [
     {limitRadius=20, drivingBone {boneName=pelvis}, sphereLocalOffset=<-5 0 0>}
     {limitRadius=13, drivingBone {boneName=lowerarm_l}, sphereLocalOffset=<30 0 0>}
    ]
   }
   n28 = Anim Dynamics - Bone: dyn_armstrap_L_03_05(ComponentPose=@n27, Alpha=1, ExternalForce="0,0,0") {
    angularDampingOverride=0.999
    bOverrideAngularDamping=true
    bUseGravityOverride=true
    bUsePlanarLimit=false
    bUseSphericalLimits=true
    simulationSpace=RootRelative
    alphaBoolBlend {blendOption=Linear}
    boundBone {boneName=dyn_armstrap_L_03_05}
    componentAppliedLinearAccClamp=<3000 3000 3000>
    componentLinearAccScale=<5 5 8>
    componentLinearVelScale=<5 5 8>
    gravityOverride {z=-980}
    physicsBodyDefinitions [{
      boundBone {boneName=dyn_armstrap_L_03_05}
      boxExtents=<15 15 15>
      constraintSetup {angularLimitsMax=<3 55 55>, angularLimitsMin=<-3 -55 -55>, angularTarget=<0 0 0>}
      localJointOffset=<10 0 0>
     }]
    relativeSpaceBone {boneName=dyn_armstrap_L_01_03}
    sphericalLimits [
     {limitRadius=20, drivingBone {boneName=pelvis}, sphereLocalOffset=<-5 0 0>}
     {limitRadius=15, drivingBone {boneName=lowerarm_l}, sphereLocalOffset=<30 0 0>}
    ]
   }
   n29 = Anim Dynamics - Bone: dyn_armstrap_R_01_01(ComponentPose=@n28, Alpha=1, ExternalForce="0,0,0") {
    angularDampingOverride=1
    bOverrideAngularDamping=true
    bUseGravityOverride=true
    bUsePlanarLimit=false
    simulationSpace=RootRelative
    alphaBoolBlend {blendOption=Linear}
    boundBone {boneName=dyn_armstrap_R_01_01}
    componentAppliedLinearAccClamp=<3000 3000 3000>
    componentLinearAccScale=<5 5 8>
    componentLinearVelScale=<5 5 8>
    gravityOverride {z=-980}
    physicsBodyDefinitions [{
      boundBone {boneName=dyn_armstrap_R_01_01}
      boxExtents=<25 25 25>
      constraintSetup {angularLimitsMax=<3 25 35>, angularLimitsMin=<-3 -25 -5>, angularTarget=<0 0 0>}
      localJointOffset=<10 0 0>
     }]
    relativeSpaceBone {boneName=spine_05}
    sphericalLimits [{limitRadius=20, drivingBone {boneName=pelvis}, sphereLocalOffset=<-5 0 0>}]
   }
   n30 = Anim Dynamics - Bone: dyn_armstrap_R_01_02(ComponentPose=@n29, Alpha=1, ExternalForce="0,0,0") {
    angularDampingOverride=0.9999
    bOverrideAngularDamping=true
    bUseGravityOverride=true
    bUsePlanarLimit=false
    bUseSphericalLimits=true
    simulationSpace=RootRelative
    alphaBoolBlend {blendOption=Linear}
    boundBone {boneName=dyn_armstrap_R_01_02}
    componentAppliedLinearAccClamp=<3000 3000 3000>
    componentLinearAccScale=<5 5 8>
    componentLinearVelScale=<5 5 8>
    gravityOverride {z=-980}
    physicsBodyDefinitions [{
      boundBone {boneName=dyn_armstrap_R_01_02}
      boxExtents=<25 25 25>
      constraintSetup {angularLimitsMax=<3 45 55>, angularLimitsMin=<-3 -45 -55>, angularTarget=<0 0 0>}
      localJointOffset=<10 0 0>
     }]
    relativeSpaceBone {boneName=dyn_armstrap_L_01_01}
    sphericalLimits [{limitRadius=20, drivingBone {boneName=pelvis}, sphereLocalOffset=<-5 0 0>}]
   }
   n31 = Anim Dynamics - Bone: dyn_armstrap_R_01_03(ComponentPose=@n30, Alpha=1, ExternalForce="0,0,0") {
    angularDampingOverride=0.999
    bOverrideAngularDamping=true
    bUseGravityOverride=true
    bUsePlanarLimit=false
    bUseSphericalLimits=true
    simulationSpace=RootRelative
    alphaBoolBlend {blendOption=Linear}
    boundBone {boneName=dyn_armstrap_R_01_03}
    componentAppliedLinearAccClamp=<3000 3000 3000>
    componentLinearAccScale=<5 5 8>
    componentLinearVelScale=<5 5 8>
    gravityOverride {z=-980}
    physicsBodyDefinitions [{
      boundBone {boneName=dyn_armstrap_R_01_03}
      boxExtents=<25 25 25>
      constraintSetup {angularLimitsMax=<3 55 55>, angularLimitsMin=<-3 -55 -55>, angularTarget=<0 0 0>}
      localJointOffset=<10 0 0>
     }]
    relativeSpaceBone {boneName=dyn_armstrap_L_01_02}
    sphericalLimits [
     {limitRadius=20, drivingBone {boneName=pelvis}, sphereLocalOffset=<-5 0 0>}
     {limitRadius=5, drivingBone {boneName=lowerarm_r}, sphereLocalOffset=<-25 0 0>}
    ]
   }
   n32 = Anim Dynamics - Bone: dyn_armstrap_R_01_04(ComponentPose=@n31, Alpha=1, ExternalForce="0,0,0") {
    angularDampingOverride=0.999
    bOverrideAngularDamping=true
    bUseGravityOverride=true
    bUsePlanarLimit=false
    bUseSphericalLimits=true
    simulationSpace=RootRelative
    alphaBoolBlend {blendOption=Linear}
    boundBone {boneName=dyn_armstrap_R_01_04}
    componentAppliedLinearAccClamp=<3000 3000 3000>
    componentLinearAccScale=<5 5 8>
    componentLinearVelScale=<5 5 8>
    gravityOverride {z=-980}
    physicsBodyDefinitions [{
      boundBone {boneName=dyn_armstrap_R_01_04}
      boxExtents=<20 20 20>
      constraintSetup {angularLimitsMax=<3 55 55>, angularLimitsMin=<-3 -55 -55>, angularTarget=<0 0 0>}
      localJointOffset=<10 0 0>
     }]
    relativeSpaceBone {boneName=dyn_armstrap_L_01_03}
    sphericalLimits [
     {limitRadius=20, drivingBone {boneName=pelvis}, sphereLocalOffset=<-5 0 0>}
     {limitRadius=13, drivingBone {boneName=lowerarm_r}, sphereLocalOffset=<-30 0 0>}
    ]
   }
   n33 = Anim Dynamics - Bone: dyn_armstrap_R_02_05(ComponentPose=@n32, Alpha=1, ExternalForce="0,0,0") {
    angularDampingOverride=0.999
    bOverrideAngularDamping=true
    bUseGravityOverride=true
    bUsePlanarLimit=false
    bUseSphericalLimits=true
    simulationSpace=RootRelative
    alphaBoolBlend {blendOption=Linear}
    boundBone {boneName=dyn_armstrap_R_02_05}
    componentAppliedLinearAccClamp=<3000 3000 3000>
    componentLinearAccScale=<5 5 8>
    componentLinearVelScale=<5 5 8>
    gravityOverride {z=-980}
    physicsBodyDefinitions [{
      boundBone {boneName=dyn_armstrap_R_02_05}
      boxExtents=<20 20 20>
      constraintSetup {angularLimitsMax=<3 55 55>, angularLimitsMin=<-3 -55 -55>, angularTarget=<0 0 0>}
      localJointOffset=<10 0 0>
     }]
    relativeSpaceBone {boneName=dyn_armstrap_L_01_03}
    sphericalLimits [
     {limitRadius=20, drivingBone {boneName=pelvis}, sphereLocalOffset=<-5 0 0>}
     {limitRadius=15, drivingBone {boneName=lowerarm_r}, sphereLocalOffset=<-30 0 0>}
    ]
   }
   n34 = Anim Dynamics - Bone: dyn_armstrap_R_02_01(ComponentPose=@n33, Alpha=1, ExternalForce="0,0,0") {
    angularDampingOverride=1
    bOverrideAngularDamping=true
    bUseGravityOverride=true
    bUsePlanarLimit=false
    simulationSpace=RootRelative
    alphaBoolBlend {blendOption=Linear}
    boundBone {boneName=dyn_armstrap_R_02_01}
    componentAppliedLinearAccClamp=<3000 3000 3000>
    componentLinearAccScale=<5 5 8>
    componentLinearVelScale=<5 5 8>
    gravityOverride {z=-980}
    physicsBodyDefinitions [{
      boundBone {boneName=dyn_armstrap_R_02_01}
      boxExtents=<25 25 25>
      constraintSetup {angularLimitsMax=<3 25 35>, angularLimitsMin=<-3 -25 -5>, angularTarget=<0 0 0>}
      localJointOffset=<10 0 0>
     }]
    relativeSpaceBone {boneName=spine_05}
    sphericalLimits [{limitRadius=20, drivingBone {boneName=pelvis}, sphereLocalOffset=<-5 0 0>}]
   }
   n35 = Anim Dynamics - Bone: dyn_armstrap_R_02_02(ComponentPose=@n34, Alpha=1, ExternalForce="0,0,0") {
    angularDampingOverride=0.9999
    bOverrideAngularDamping=true
    bUseGravityOverride=true
    bUsePlanarLimit=false
    bUseSphericalLimits=true
    simulationSpace=RootRelative
    alphaBoolBlend {blendOption=Linear}
    boundBone {boneName=dyn_armstrap_R_02_02}
    componentAppliedLinearAccClamp=<3000 3000 3000>
    componentLinearAccScale=<5 5 8>
    componentLinearVelScale=<5 5 8>
    gravityOverride {z=-980}
    physicsBodyDefinitions [{
      boundBone {boneName=dyn_armstrap_R_02_02}
      boxExtents=<25 25 25>
      constraintSetup {angularLimitsMax=<3 45 55>, angularLimitsMin=<-3 -45 -55>, angularTarget=<0 0 0>}
      localJointOffset=<10 0 0>
     }]
    relativeSpaceBone {boneName=dyn_armstrap_L_01_01}
    sphericalLimits [{limitRadius=20, drivingBone {boneName=pelvis}, sphereLocalOffset=<-5 0 0>}]
   }
   n36 = Anim Dynamics - Bone: dyn_armstrap_R_02_03(ComponentPose=@n35, Alpha=1, ExternalForce="0,0,0") {
    angularDampingOverride=0.999
    bOverrideAngularDamping=true
    bUseGravityOverride=true
    bUsePlanarLimit=false
    bUseSphericalLimits=true
    simulationSpace=RootRelative
    alphaBoolBlend {blendOption=Linear}
    boundBone {boneName=dyn_armstrap_R_02_03}
    componentAppliedLinearAccClamp=<3000 3000 3000>
    componentLinearAccScale=<5 5 8>
    componentLinearVelScale=<5 5 8>
    gravityOverride {z=-980}
    physicsBodyDefinitions [{
      boundBone {boneName=dyn_armstrap_R_02_03}
      boxExtents=<25 25 25>
      constraintSetup {angularLimitsMax=<3 55 55>, angularLimitsMin=<-3 -55 -55>, angularTarget=<0 0 0>}
      localJointOffset=<10 0 0>
     }]
    relativeSpaceBone {boneName=dyn_armstrap_L_01_02}
    sphericalLimits [{limitRadius=20, drivingBone {boneName=pelvis}, sphereLocalOffset=<-5 0 0>}]
   }
   n37 = Anim Dynamics - Bone: dyn_armstrap_R_02_04(ComponentPose=@n36, Alpha=1, ExternalForce="0,0,0") {
    angularDampingOverride=0.99
    bOverrideAngularDamping=true
    bUseGravityOverride=true
    bUsePlanarLimit=false
    bUseSphericalLimits=true
    simulationSpace=RootRelative
    alphaBoolBlend {blendOption=Linear}
    boundBone {boneName=dyn_armstrap_R_02_04}
    componentAppliedLinearAccClamp=<3000 3000 3000>
    componentLinearAccScale=<6 6 9>
    componentLinearVelScale=<6 6 9>
    gravityOverride {z=-980}
    physicsBodyDefinitions [{
      boundBone {boneName=dyn_armstrap_R_02_04}
      boxExtents=<20 20 20>
      constraintSetup {angularLimitsMax=<3 55 55>, angularLimitsMin=<-3 -55 -55>, angularTarget=<0 0 0>}
      localJointOffset=<10 0 0>
     }]
    relativeSpaceBone {boneName=dyn_armstrap_L_01_03}
    sphericalLimits [
     {limitRadius=20, drivingBone {boneName=pelvis}, sphereLocalOffset=<-5 0 0>}
     {limitRadius=13, drivingBone {boneName=lowerarm_r}, sphereLocalOffset=<-30 0 0>}
    ]
   }
   n38 = Anim Dynamics - Bone: dyn_armstrap_R_02_05(ComponentPose=@n37, Alpha=1, ExternalForce="0,0,0") {
    angularDampingOverride=0.99
    bOverrideAngularDamping=true
    bUseGravityOverride=true
    bUsePlanarLimit=false
    bUseSphericalLimits=true
    simulationSpace=RootRelative
    alphaBoolBlend {blendOption=Linear}
    boundBone {boneName=dyn_armstrap_R_02_05}
    componentAppliedLinearAccClamp=<3000 3000 3000>
    componentLinearAccScale=<6 6 9>
    componentLinearVelScale=<6 6 9>
    gravityOverride {z=-980}
    physicsBodyDefinitions [{
      boundBone {boneName=dyn_armstrap_R_02_05}
      boxExtents=<20 20 20>
      constraintSetup {angularLimitsMax=<3 55 55>, angularLimitsMin=<-3 -55 -55>, angularTarget=<0 0 0>}
      localJointOffset=<10 0 0>
     }]
    relativeSpaceBone {boneName=dyn_armstrap_L_01_03}
    sphericalLimits [
     {limitRadius=20, drivingBone {boneName=pelvis}, sphereLocalOffset=<-5 0 0>}
     {limitRadius=15, drivingBone {boneName=lowerarm_r}, sphereLocalOffset=<-30 0 0>}
    ]
   }
   n39 = Anim Dynamics - Bone: dyn_armstrap_R_03_01(ComponentPose=@n38, Alpha=1, ExternalForce="0,0,0") {
    angularDampingOverride=1
    bOverrideAngularDamping=true
    bUseGravityOverride=true
    bUsePlanarLimit=false
    simulationSpace=RootRelative
    alphaBoolBlend {blendOption=Linear}
    boundBone {boneName=dyn_armstrap_R_03_01}
    componentAppliedLinearAccClamp=<3000 3000 3000>
    componentLinearAccScale=<5 5 8>
    componentLinearVelScale=<5 5 8>
    gravityOverride {z=-980}
    physicsBodyDefinitions [{
      boundBone {boneName=dyn_armstrap_R_03_01}
      boxExtents=<10 25 25>
      constraintSetup {angularLimitsMax=<3 0 55>, angularLimitsMin=<-3 0 0>, angularTarget=<0 0 0>}
      localJointOffset=<5 0 0>
     }]
    relativeSpaceBone {boneName=spine_05}
    sphericalLimits [{limitRadius=20, drivingBone {boneName=pelvis}, sphereLocalOffset=<-5 0 0>}]
   }
   n40 = Anim Dynamics - Bone: dyn_armstrap_R_03_02(ComponentPose=@n39, Alpha=1, ExternalForce="0,0,0") {
    angularDampingOverride=0.9999
    bOverrideAngularDamping=true
    bUseGravityOverride=true
    bUsePlanarLimit=false
    bUseSphericalLimits=true
    simulationSpace=RootRelative
    alphaBoolBlend {blendOption=Linear}
    boundBone {boneName=dyn_armstrap_R_03_02}
    componentAppliedLinearAccClamp=<3000 3000 3000>
    componentLinearAccScale=<5 5 8>
    componentLinearVelScale=<5 5 8>
    gravityOverride {z=-980}
    physicsBodyDefinitions [{
      boundBone {boneName=dyn_armstrap_R_03_02}
      boxExtents=<15 25 25>
      constraintSetup {angularLimitsMax=<3 45 55>, angularLimitsMin=<-3 -45 0>, angularTarget=<0 0 0>}
      localJointOffset=<5 0 0>
     }]
    relativeSpaceBone {boneName=dyn_armstrap_L_01_01}
    sphericalLimits [{limitRadius=20, drivingBone {boneName=pelvis}, sphereLocalOffset=<-5 0 0>}]
   }
   n41 = Anim Dynamics - Bone: dyn_armstrap_R_03_03(ComponentPose=@n40, Alpha=1, ExternalForce="0,0,0") {
    angularDampingOverride=0.999
    bOverrideAngularDamping=true
    bUseGravityOverride=true
    bUsePlanarLimit=false
    bUseSphericalLimits=true
    simulationSpace=RootRelative
    alphaBoolBlend {blendOption=Linear}
    boundBone {boneName=dyn_armstrap_R_03_03}
    componentAppliedLinearAccClamp=<3000 3000 3000>
    componentLinearAccScale=<5 5 8>
    componentLinearVelScale=<5 5 8>
    gravityOverride {z=-980}
    physicsBodyDefinitions [{
      boundBone {boneName=dyn_armstrap_R_03_03}
      boxExtents=<15 25 25>
      constraintSetup {angularLimitsMax=<3 55 55>, angularLimitsMin=<-3 -55 -55>, angularTarget=<0 0 0>}
      localJointOffset=<10 0 0>
     }]
    relativeSpaceBone {boneName=dyn_armstrap_L_01_02}
    sphericalLimits [{limitRadius=20, drivingBone {boneName=pelvis}, sphereLocalOffset=<-5 0 0>}]
   }
   n42 = Anim Dynamics - Bone: dyn_armstrap_R_03_04(ComponentPose=@n41, Alpha=1, ExternalForce="0,0,0") {
    angularDampingOverride=0.999
    bOverrideAngularDamping=true
    bUseGravityOverride=true
    bUsePlanarLimit=false
    bUseSphericalLimits=true
    simulationSpace=RootRelative
    alphaBoolBlend {blendOption=Linear}
    boundBone {boneName=dyn_armstrap_R_03_04}
    componentAppliedLinearAccClamp=<3000 3000 3000>
    componentLinearAccScale=<5 5 8>
    componentLinearVelScale=<5 5 8>
    gravityOverride {z=-980}
    physicsBodyDefinitions [{
      boundBone {boneName=dyn_armstrap_R_03_04}
      boxExtents=<20 20 20>
      constraintSetup {angularLimitsMax=<3 55 55>, angularLimitsMin=<-3 -55 -55>, angularTarget=<0 0 0>}
      localJointOffset=<15 0 0>
     }]
    relativeSpaceBone {boneName=dyn_armstrap_L_01_03}
    sphericalLimits [
     {limitRadius=20, drivingBone {boneName=pelvis}, sphereLocalOffset=<-5 0 0>}
     {limitRadius=13, drivingBone {boneName=lowerarm_r}, sphereLocalOffset=<-28 0 0>}
    ]
   }
   n43 = Anim Dynamics - Bone: dyn_armstrap_R_03_05(ComponentPose=@n42, Alpha=1, ExternalForce="0,0,0") {
    angularDampingOverride=0.999
    bOverrideAngularDamping=true
    bUseGravityOverride=true
    bUsePlanarLimit=false
    bUseSphericalLimits=true
    simulationSpace=RootRelative
    alphaBoolBlend {blendOption=Linear}
    boundBone {boneName=dyn_armstrap_R_03_05}
    componentAppliedLinearAccClamp=<3000 3000 3000>
    componentLinearAccScale=<5 5 8>
    componentLinearVelScale=<5 5 8>
    gravityOverride {z=-980}
    physicsBodyDefinitions [{
      boundBone {boneName=dyn_armstrap_R_03_05}
      boxExtents=<20 20 20>
      constraintSetup {angularLimitsMax=<3 55 55>, angularLimitsMin=<-3 -55 -55>, angularTarget=<0 0 0>}
      localJointOffset=<15 0 0>
     }]
    relativeSpaceBone {boneName=dyn_armstrap_L_01_03}
    sphericalLimits [
     {limitRadius=20, drivingBone {boneName=pelvis}, sphereLocalOffset=<-5 0 0>}
     {limitRadius=15, drivingBone {boneName=lowerarm_r}, sphereLocalOffset=<-30 0 0>}
    ]
   }
   n44 = Trail controller - Bone: dyn_armstrap_L_01_07(ComponentPose=@n43, Alpha=0.7) {
    chainLength=3
    relaxationSpeedScale=1.5
    planarLimits [{drivingBone {boneName=thigh_l}, planeTransform {location=<0 0 -40>, rotation=<rot -90 90 0>, scale=<0.2 1 1>}}]
    rotationLimits [{limitMax=<90 90 90>, limitMin=<-90 -90 -90>}, {limitMax=<90 90 90>, limitMin=<-90 -90 -90>}, {limitMax=<90 90 90>, limitMin=<-90 -90 -90>}]
    rotationOffsets [<0 0 0>, <0 0 0>, <0 0 0>]
    trailBone {boneName=dyn_armstrap_L_01_07}
    trailRelaxationSpeed {editorCurveData {keys [<at 0 10 RCIM_Linear>,<at 1 4 RCIM_Linear>]}}
   }
   n45 = Trail controller - Bone: dyn_armstrap_L_02_07(ComponentPose=@n44, Alpha=0.7) {
    chainLength=3
    relaxationSpeedScale=1.5
    rotationLimits [{limitMax=<90 90 90>, limitMin=<-90 -90 -90>}, {limitMax=<90 90 90>, limitMin=<-90 -90 -90>}, {limitMax=<90 90 90>, limitMin=<-90 -90 -90>}]
    rotationOffsets [<0 0 0>, <0 0 0>, <0 0 0>]
    trailBone {boneName=dyn_armstrap_L_02_07}
    trailRelaxationSpeed {editorCurveData {keys [<at 0 10 RCIM_Linear>,<at 1 4 RCIM_Linear>]}}
   }
   n46 = Trail controller - Bone: dyn_armstrap_L_03_06(ComponentPose=@n45, Alpha=0.7) {
    chainLength=3
    relaxationSpeedScale=1.5
    rotationLimits [{limitMax=<90 90 90>, limitMin=<-90 -90 -90>}, {limitMax=<90 90 90>, limitMin=<-90 -90 -90>}, {limitMax=<90 90 90>, limitMin=<-90 -90 -90>}]
    rotationOffsets [<0 0 0>, <0 0 0>, <0 0 0>]
    trailBone {boneName=dyn_armstrap_L_03_06}
    trailRelaxationSpeed {editorCurveData {keys [<at 0 10 RCIM_Linear>,<at 1 4 RCIM_Linear>]}}
   }
   n47 = Trail controller - Bone: dyn_armstrap_R_01_07(ComponentPose=@n46, Alpha=0.7) {
    chainLength=3
    relaxationSpeedScale=1.5
    rotationLimits [{limitMax=<180 180 180>}, {limitMax=<180 180 180>}, {limitMax=<180 180 180>}]
    rotationOffsets [<0 0 0>, <0 0 0>, <0 0 0>]
    trailBone {boneName=dyn_armstrap_R_01_07}
   }
   n48 = Trail controller - Bone: dyn_armstrap_R_02_07(ComponentPose=@n47, Alpha=0.7) {
    chainLength=3
    relaxationSpeedScale=1.5
    rotationLimits [{limitMax=<180 180 180>}, {limitMax=<180 180 180>}, {limitMax=<180 180 180>}]
    rotationOffsets [<0 0 0>, <0 0 0>, <0 0 0>]
    trailBone {boneName=dyn_armstrap_R_02_07}
   }
   n49 = Trail controller - Bone: dyn_armstrap_R_03_06(ComponentPose=@n48, Alpha=0.7) {
    chainLength=3
    relaxationSpeedScale=1.5
    rotationLimits [{limitMax=<180 180 180>}, {limitMax=<180 180 180>}, {limitMax=<180 180 180>}]
    rotationOffsets [<0 0 0>, <0 0 0>, <0 0 0>]
    trailBone {boneName=dyn_armstrap_R_03_06}
   }
   n50 = Trail controller - Bone: dyn_armstrap_L_01_07(ComponentPose=@n49, Alpha=0.7) {
    relaxationSpeedScale=1.5
    planarLimits [{drivingBone {boneName=thigh_l}, planeTransform {location=<0 0 -40>, rotation=<rot -90 90 0>, scale=<0.2 1 1>}}]
    rotationLimits [{limitMax=<90 90 90>, limitMin=<-90 -90 -90>}, {limitMax=<90 90 90>, limitMin=<-90 -90 -90>}]
    rotationOffsets [<0 0 0>, <0 0 0>]
    trailBone {boneName=dyn_armstrap_L_01_07}
    trailRelaxationSpeed {editorCurveData {keys [<at 0 10 RCIM_Linear>,<at 1 4 RCIM_Linear>]}}
   }
   n51 = Trail controller - Bone: dyn_armstrap_L_02_07(ComponentPose=@n50, Alpha=0.7) {
    relaxationSpeedScale=1.5
    rotationLimits [{limitMax=<90 90 90>, limitMin=<-90 -90 -90>}, {limitMax=<90 90 90>, limitMin=<-90 -90 -90>}]
    rotationOffsets [<0 0 0>, <0 0 0>]
    trailBone {boneName=dyn_armstrap_L_02_07}
    trailRelaxationSpeed {editorCurveData {keys [<at 0 10 RCIM_Linear>,<at 1 4 RCIM_Linear>]}}
   }
   n52 = Trail controller - Bone: dyn_armstrap_L_03_06(ComponentPose=@n51, Alpha=0.7) {
    relaxationSpeedScale=1.5
    rotationLimits [{limitMax=<90 90 90>, limitMin=<-90 -90 -90>}, {limitMax=<90 90 90>, limitMin=<-90 -90 -90>}]
    rotationOffsets [<0 0 0>, <0 0 0>]
    trailBone {boneName=dyn_armstrap_L_03_06}
    trailRelaxationSpeed {editorCurveData {keys [<at 0 10 RCIM_Linear>,<at 1 4 RCIM_Linear>]}}
   }
   n53 = Trail controller - Bone: dyn_armstrap_R_01_07(ComponentPose=@n52, Alpha=0.7) {
    relaxationSpeedScale=1.5
    rotationLimits [{limitMax=<180 180 180>}, {limitMax=<180 180 180>}]
    rotationOffsets [<0 0 0>, <0 0 0>]
    trailBone {boneName=dyn_armstrap_R_01_07}
   }
   n54 = Trail controller - Bone: dyn_armstrap_R_02_07(ComponentPose=@n53, Alpha=0.7) {
    relaxationSpeedScale=1.5
    rotationLimits [{limitMax=<180 180 180>}, {limitMax=<180 180 180>}]
    rotationOffsets [<0 0 0>, <0 0 0>]
    trailBone {boneName=dyn_armstrap_R_02_07}
   }
   n55 = Trail controller - Bone: dyn_armstrap_R_03_06(ComponentPose=@n54, Alpha=0.7) {
    relaxationSpeedScale=1.5
    rotationLimits [{limitMax=<180 180 180>}, {limitMax=<180 180 180>}]
    rotationOffsets [<0 0 0>, <0 0 0>]
    trailBone {boneName=dyn_armstrap_R_03_06}
   }
   n56 = "Constraint - dyn_armstrap_L_04_mstr (PPPPP)"!(
    ComponentPose=@n55
    Alpha=1
    ConstraintWeights_0=0.11
    ConstraintWeights_1:float=0.55
    ConstraintWeights_2:float=0.09
    ConstraintWeights_3:float=0.17
    ConstraintWeights_4:float=0.08
   ) -> (Pose) {
    alphaBoolBlend {blendOption=Linear}
    boneToModify {boneName=dyn_armstrap_L_04_mstr}
    constraintSetup [
     {targetBone {boneName=upperarm_twist_01_l}}
     {targetBone {boneName=upperarm_l}}
     {targetBone {boneName=lowerarm_twist_02_l}}
     {targetBone {boneName=upperarm_twist_02_l}}
     {targetBone {boneName=lowerarm_l}}
    ]
    constraintWeights [1,1,1,1,1]
   }
   n57 = Anim Dynamics - Bone: dyn_armstrap_L_04_01(ComponentPose=@n56, Alpha=1, ExternalForce="0,0,0") {
    angularDampingOverride=1
    angularSpringConstant=700
    bAngularSpring=true
    bOverrideAngularDamping=true
    bUseGravityOverride=true
    simulationSpace=RootRelative
    alphaBoolBlend {blendOption=Linear}
    boundBone {boneName=dyn_armstrap_L_04_01}
    componentAppliedLinearAccClamp=<10000 10000 10000>
    componentLinearAccScale=<0.1 0.1 0.3>
    componentLinearVelScale=<0.1 0.1 0.3>
    gravityOverride {z=-980}
    physicsBodyDefinitions [{
      boundBone {boneName=dyn_armstrap_L_04_01}
      boxExtents=<15 15 40>
      constraintSetup {angularLimitsMax=<0 30 50>, angularLimitsMin=<0 -30 -30>, angularTarget=<0 0 0>}
      localJointOffset=<0.5 0 0>
     }]
   }
   n58 = Anim Dynamics - Bone: dyn_armstrap_L_04_02(ComponentPose=@n57, Alpha=1, ExternalForce="0,0,0") {
    angularDampingOverride=1
    angularSpringConstant=700
    bAngularSpring=true
    bOverrideAngularDamping=true
    bUseGravityOverride=true
    simulationSpace=RootRelative
    alphaBoolBlend {blendOption=Linear}
    boundBone {boneName=dyn_armstrap_L_04_02}
    componentAppliedLinearAccClamp=<10000 10000 10000>
    componentLinearAccScale=<0.1 0.1 0.3>
    componentLinearVelScale=<0.1 0.1 0.3>
    gravityOverride {z=-980}
    physicsBodyDefinitions [{
      boundBone {boneName=dyn_armstrap_L_04_02}
      boxExtents=<5 5 20>
      constraintSetup {angularLimitsMax=<10 15 15>, angularLimitsMin=<-10 -15 -15>, angularTarget=<0 0 0>}
      localJointOffset=<0.5 0 0>
     }]
   }
   n59 = Anim Dynamics - Bone: dyn_armstrap_L_04_03(ComponentPose=@n58, Alpha=1, ExternalForce="0,0,0") {
    angularDampingOverride=1
    angularSpringConstant=700
    bAngularSpring=true
    bOverrideAngularDamping=true
    bUseGravityOverride=true
    simulationSpace=RootRelative
    alphaBoolBlend {blendOption=Linear}
    boundBone {boneName=dyn_armstrap_L_04_03}
    componentAppliedLinearAccClamp=<10000 10000 10000>
    componentLinearAccScale=<0.1 0.1 0.3>
    componentLinearVelScale=<0.1 0.1 0.3>
    gravityOverride {z=-980}
    physicsBodyDefinitions [{
      boundBone {boneName=dyn_armstrap_L_04_03}
      boxExtents=<5 5 20>
      constraintSetup {angularLimitsMax=<10 15 15>, angularLimitsMin=<-10 -15 -15>, angularTarget=<0 0 0>}
      localJointOffset=<0.5 0 0>
     }]
   }
   n60 = "Constraint - dyn_armstrap_R_04_mstr (PPPPP)"!(
    ComponentPose=@n59
    Alpha=1
    ConstraintWeights_0=0.26
    ConstraintWeights_1:float=0.02
    ConstraintWeights_2:float=0.34
    ConstraintWeights_3:float=0.36
    ConstraintWeights_4:float=0.02
   ) -> (Pose) {
    alphaBoolBlend {blendOption=Linear}
    boneToModify {boneName=dyn_armstrap_R_04_mstr}
    constraintSetup [
     {targetBone {boneName=upperarm_twist_01_r}}
     {targetBone {boneName=lowerarm_twist_02_r}}
     {targetBone {boneName=upperarm_r}}
     {targetBone {boneName=upperarm_twist_02_r}}
     {targetBone {boneName=lowerarm_r}}
    ]
    constraintWeights [1,1,1,1,1]
   }
   n61 = Anim Dynamics - Bone: dyn_armstrap_R_04_01(ComponentPose=@n60, Alpha=1, ExternalForce="0,0,0") {
    angularDampingOverride=1
    angularSpringConstant=700
    bAngularSpring=true
    bOverrideAngularDamping=true
    bUseGravityOverride=true
    simulationSpace=RootRelative
    alphaBoolBlend {blendOption=Linear}
    boundBone {boneName=dyn_armstrap_R_04_01}
    componentAppliedLinearAccClamp=<10000 10000 10000>
    componentLinearAccScale=<0.1 0.1 0.3>
    componentLinearVelScale=<0.1 0.1 0.3>
    gravityOverride {z=-980}
    physicsBodyDefinitions [{
      boundBone {boneName=dyn_armstrap_R_04_01}
      boxExtents=<15 15 40>
      constraintSetup {angularLimitsMax=<0 30 30>, angularLimitsMin=<0 -30 -30>, angularTarget=<0 0 0>}
      localJointOffset=<0.5 0 0>
     }]
   }
   n62 = Anim Dynamics - Bone: dyn_armstrap_R_04_02(ComponentPose=@n61, Alpha=1, ExternalForce="0,0,0") {
    angularDampingOverride=1
    angularSpringConstant=700
    bAngularSpring=true
    bOverrideAngularDamping=true
    bUseGravityOverride=true
    simulationSpace=RootRelative
    alphaBoolBlend {blendOption=Linear}
    boundBone {boneName=dyn_armstrap_R_04_02}
    componentAppliedLinearAccClamp=<10000 10000 10000>
    componentLinearAccScale=<0.1 0.1 0.3>
    componentLinearVelScale=<0.1 0.1 0.3>
    gravityOverride {z=-980}
    physicsBodyDefinitions [{
      boundBone {boneName=dyn_armstrap_R_04_02}
      boxExtents=<5 5 20>
      constraintSetup {angularLimitsMax=<10 15 15>, angularLimitsMin=<-10 -15 -15>, angularTarget=<0 0 0>}
      localJointOffset=<0.5 0 0>
     }]
   }
   n63 = Anim Dynamics - Bone: dyn_armstrap_R_04_03(ComponentPose=@n62, Alpha=1, ExternalForce="0,0,0") {
    angularDampingOverride=1
    angularSpringConstant=700
    bAngularSpring=true
    bOverrideAngularDamping=true
    bUseGravityOverride=true
    simulationSpace=RootRelative
    alphaBoolBlend {blendOption=Linear}
    boundBone {boneName=dyn_armstrap_R_04_03}
    componentAppliedLinearAccClamp=<10000 10000 10000>
    componentLinearAccScale=<0.1 0.1 0.3>
    componentLinearVelScale=<0.1 0.1 0.3>
    gravityOverride {z=-980}
    physicsBodyDefinitions [{
      boundBone {boneName=dyn_armstrap_R_04_03}
      boxExtents=<5 5 20>
      constraintSetup {angularLimitsMax=<10 15 15>, angularLimitsMin=<-10 -15 -15>, angularTarget=<0 0 0>}
      localJointOffset=<0.5 0 0>
     }]
   }
   n64 = Anim Dynamics - Bone: dyn_tassel_01(ComponentPose=@n63, Alpha=1, ExternalForce="0,0,0") {
    angularDampingOverride=0.999
    angularSpringConstant=700
    bAngularSpring=true
    bOverrideAngularDamping=true
    bUseGravityOverride=true
    bUsePlanarLimit=false
    simulationSpace=RootRelative
    alphaBoolBlend {blendOption=Linear}
    boundBone {boneName=dyn_tassel_01}
    componentAppliedLinearAccClamp=<10000 10000 10000>
    componentLinearAccScale=<0.1 0.1 0.3>
    componentLinearVelScale=<0.1 0.1 0.3>
    gravityOverride {z=-980}
    physicsBodyDefinitions [{
      boundBone {boneName=dyn_tassel_01}
      boxExtents=<5 15 5>
      constraintSetup {angularLimitsMax=<0 30 30>, angularLimitsMin=<0 -30 -30>, angularTarget=<0 0 0>}
      localJointOffset=<1 0 0>
     }]
    planarLimits [{drivingBone {boneName=thigh_r}, planeTransform {location=<10 -5 0>, rotation=<rot 0 0 -110>}}]
   }
   n65 = Anim Dynamics - Bone: dyn_tassel_02(ComponentPose=@n64, Alpha=1, ExternalForce="0,0,0") {
    angularDampingOverride=0.999
    angularSpringConstant=700
    bAngularSpring=true
    bOverrideAngularDamping=true
    bUseGravityOverride=true
    bUsePlanarLimit=false
    simulationSpace=RootRelative
    alphaBoolBlend {blendOption=Linear}
    boundBone {boneName=dyn_tassel_02}
    componentAppliedLinearAccClamp=<10000 10000 10000>
    componentLinearAccScale=<0.1 0.1 0.3>
    componentLinearVelScale=<0.1 0.1 0.3>
    gravityOverride {z=-980}
    physicsBodyDefinitions [{
      boundBone {boneName=dyn_tassel_02}
      boxExtents=<5 10 5>
      constraintSetup {angularLimitsMax=<0 30 30>, angularLimitsMin=<0 -30 -30>, angularTarget=<0 0 0>}
      localJointOffset=<1 0 0>
     }]
    planarLimits [{drivingBone {boneName=thigh_r}, planeTransform {location=<10 -5 0>, rotation=<rot 0 0 -110>}}]
   }
   n66 = Anim Dynamics - Bone: dyn_tassel_03(ComponentPose=@n65, Alpha=1, ExternalForce="0,0,0") {
    angularDampingOverride=0.999
    angularSpringConstant=700
    bAngularSpring=true
    bOverrideAngularDamping=true
    bUseGravityOverride=true
    bUsePlanarLimit=false
    simulationSpace=RootRelative
    alphaBoolBlend {blendOption=Linear}
    boundBone {boneName=dyn_tassel_03}
    componentAppliedLinearAccClamp=<10000 10000 10000>
    componentLinearAccScale=<0.1 0.1 0.3>
    componentLinearVelScale=<0.1 0.1 0.3>
    gravityOverride {z=-980}
    physicsBodyDefinitions [{
      boundBone {boneName=dyn_tassel_03}
      boxExtents=<5 10 5>
      constraintSetup {angularLimitsMax=<0 30 30>, angularLimitsMin=<0 -30 -30>, angularTarget=<0 0 0>}
      localJointOffset=<1 0 0>
     }]
    planarLimits [{drivingBone {boneName=thigh_r}, planeTransform {location=<10 -5 0>, rotation=<rot 0 0 -110>}}]
   }
   n67 = Anim Dynamics - Bone: dyn_fold(ComponentPose=@n66, Alpha=1, ExternalForce="0,0,0") {
    angularDampingOverride=0.999
    angularSpringConstant=10000
    bAngularSpring=true
    bOverrideAngularDamping=true
    bUseGravityOverride=true
    simulationSpace=RootRelative
    alphaBoolBlend {blendOption=Linear}
    boundBone {boneName=dyn_fold}
    componentAppliedLinearAccClamp=<10000 10000 10000>
    componentLinearAccScale=<0.1 0.1 0.3>
    componentLinearVelScale=<0.1 0.1 0.3>
    gravityOverride {z=-980}
    physicsBodyDefinitions [{
      boundBone {boneName=dyn_fold}
      boxExtents=<10 10 10>
      constraintSetup {angularLimitsMax=<0 5 0>, angularLimitsMin=<0 -20 0>, angularTarget=<0 0 0>}
      localJointOffset=<2 0 0>
     }]
   }
   n68 = "Constraint - dyn_lace_l_01_mstr (PPP)"!(ComponentPose=@n67, Alpha=1, ConstraintWeights_0=0.02, ConstraintWeights_1:float=0.03, ConstraintWeights_2:float=0.95) -> (Pose) {
    alphaBoolBlend {blendOption=Linear}
    boneToModify {boneName=dyn_lace_l_01_mstr}
    constraintSetup [{targetBone {boneName=calf_twist_01_l}}, {targetBone {boneName=calf_twist_02_l}}, {targetBone {boneName=foot_l}}]
    constraintWeights [1,1,1]
   }
   n69 = Anim Dynamics - Bone: dyn_lace_l_01_01(ComponentPose=@n68, Alpha=1, ExternalForce="0,0,0") {
    angularDampingOverride=0.9
    bOverrideAngularDamping=true
    simulationSpace=RootRelative
    alphaBoolBlend {blendOption=Linear}
    boundBone {boneName=dyn_lace_l_01_01}
    componentAppliedLinearAccClamp=<10000 10000 10000>
    componentLinearAccScale=<5 5 5>
    componentLinearVelScale=<5 5 5>
    physicsBodyDefinitions [{
      boundBone {boneName=dyn_lace_l_01_01}
      boxExtents=<30 30 30>
      constraintSetup {angularLimitsMax=<0 2 25>, angularLimitsMin=<0 -20 -25>, angularTarget=<0 0 0>}
      localJointOffset=<0.5 0 0>
     }]
    relativeSpaceBone {boneName=foot_l}
   }
   n70 = Anim Dynamics - Bone: dyn_lace_l_01_02(ComponentPose=@n69, Alpha=1, ExternalForce="0,0,0") {
    angularDampingOverride=0.9
    bOverrideAngularDamping=true
    simulationSpace=RootRelative
    alphaBoolBlend {blendOption=Linear}
    boundBone {boneName=dyn_lace_l_01_02}
    componentAppliedLinearAccClamp=<10000 10000 10000>
    componentLinearAccScale=<3 3 3>
    componentLinearVelScale=<3 3 3>
    physicsBodyDefinitions [{
      boundBone {boneName=dyn_lace_l_01_02}
      boxExtents=<15 15 15>
      constraintSetup {angularLimitsMax=<0 2 25>, angularLimitsMin=<0 -20 -25>, angularTarget=<0 0 0>}
      localJointOffset=<0.5 0 0>
     }]
    relativeSpaceBone {boneName=foot_l}
   }
   n71 = Anim Dynamics - Bone: dyn_lace_l_01_03(ComponentPose=@n70, Alpha=1, ExternalForce="0,0,0") {
    angularDampingOverride=0.9
    bOverrideAngularDamping=true
    simulationSpace=RootRelative
    alphaBoolBlend {blendOption=Linear}
    boundBone {boneName=dyn_lace_l_01_03}
    componentAppliedLinearAccClamp=<10000 10000 10000>
    componentLinearAccScale=<2 2 2>
    componentLinearVelScale=<2 2 2>
    physicsBodyDefinitions [{
      boundBone {boneName=dyn_lace_l_01_03}
      boxExtents=<10 10 10>
      constraintSetup {angularLimitsMax=<0 2 25>, angularLimitsMin=<0 -20 -25>, angularTarget=<0 0 0>}
      localJointOffset=<0.5 0 0>
     }]
    relativeSpaceBone {boneName=foot_l}
   }
   n72 = "Constraint - dyn_lace_l_02_mstr (PPP)"!(ComponentPose=@n71, Alpha=1, ConstraintWeights_0=0.02, ConstraintWeights_1:float=0.03, ConstraintWeights_2:float=0.95) -> (Pose) {
    alphaBoolBlend {blendOption=Linear}
    boneToModify {boneName=dyn_lace_l_02_mstr}
    constraintSetup [{targetBone {boneName=calf_twist_01_l}}, {targetBone {boneName=calf_twist_02_l}}, {targetBone {boneName=foot_l}}]
    constraintWeights [1,1,1]
   }
   n73 = Anim Dynamics - Bone: dyn_lace_l_02_01(ComponentPose=@n72, Alpha=1, ExternalForce="0,0,0") {
    angularDampingOverride=0.9
    angularSpringConstant=0.9999
    bAngularSpring=true
    bOverrideAngularDamping=true
    simulationSpace=RootRelative
    alphaBoolBlend {blendOption=Linear}
    boundBone {boneName=dyn_lace_l_02_01}
    componentAppliedLinearAccClamp=<10000 10000 10000>
    componentLinearAccScale=<5 5 5>
    componentLinearVelScale=<5 5 5>
    physicsBodyDefinitions [{
      boundBone {boneName=dyn_lace_l_02_01}
      boxExtents=<30 30 30>
      constraintSetup {angularLimitsMax=<0 2 25>, angularLimitsMin=<0 -20 -25>, angularTarget=<0 0 0>}
      localJointOffset=<0.5 0 0>
     }]
    relativeSpaceBone {boneName=foot_l}
   }
   n74 = Anim Dynamics - Bone: dyn_lace_l_02_02(ComponentPose=@n73, Alpha=1, ExternalForce="0,0,0") {
    angularDampingOverride=0.9
    angularSpringConstant=0.9999
    bAngularSpring=true
    bOverrideAngularDamping=true
    simulationSpace=RootRelative
    alphaBoolBlend {blendOption=Linear}
    boundBone {boneName=dyn_lace_l_02_02}
    componentAppliedLinearAccClamp=<10000 10000 10000>
    componentLinearAccScale=<3 3 3>
    componentLinearVelScale=<3 3 3>
    physicsBodyDefinitions [{
      boundBone {boneName=dyn_lace_l_02_02}
      boxExtents=<15 15 15>
      constraintSetup {angularLimitsMax=<0 2 25>, angularLimitsMin=<0 -20 -25>, angularTarget=<0 0 0>}
      localJointOffset=<0.5 0 0>
     }]
    relativeSpaceBone {boneName=foot_l}
   }
   n75 = Anim Dynamics - Bone: dyn_lace_l_02_03(ComponentPose=@n74, Alpha=1, ExternalForce="0,0,0") {
    angularDampingOverride=0.8
    angularSpringConstant=0.9999
    bAngularSpring=true
    bOverrideAngularDamping=true
    simulationSpace=RootRelative
    alphaBoolBlend {blendOption=Linear}
    boundBone {boneName=dyn_lace_l_02_03}
    componentAppliedLinearAccClamp=<10000 10000 10000>
    componentLinearAccScale=<2 2 2>
    componentLinearVelScale=<2 2 2>
    physicsBodyDefinitions [{
      boundBone {boneName=dyn_lace_l_02_03}
      boxExtents=<10 10 10>
      constraintSetup {angularLimitsMax=<0 2 25>, angularLimitsMin=<0 -20 -25>, angularTarget=<0 0 0>}
      localJointOffset=<0.5 0 0>
     }]
    relativeSpaceBone {boneName=foot_l}
   }
   n76 = "Constraint - dyn_lace_r_01_mstr (PPP)"!(ComponentPose=@n75, Alpha=1, ConstraintWeights_0=0.02, ConstraintWeights_1:float=0.03, ConstraintWeights_2:float=0.95) -> (Pose) {
    alphaBoolBlend {blendOption=Linear}
    boneToModify {boneName=dyn_lace_r_01_mstr}
    constraintSetup [{targetBone {boneName=calf_twist_01_r}}, {targetBone {boneName=calf_twist_02_r}}, {targetBone {boneName=foot_r}}]
    constraintWeights [1,1,1]
   }
   n77 = Anim Dynamics - Bone: dyn_lace_r_01_01(ComponentPose=@n76, Alpha=1, ExternalForce="0,0,0") {
    angularDampingOverride=0.9
    angularSpringConstant=0.99
    bAngularSpring=true
    bOverrideAngularDamping=true
    simulationSpace=RootRelative
    alphaBoolBlend {blendOption=Linear}
    boundBone {boneName=dyn_lace_r_01_01}
    componentAppliedLinearAccClamp=<10000 10000 10000>
    componentLinearAccScale=<5 5 5>
    componentLinearVelScale=<5 5 5>
    physicsBodyDefinitions [{
      boundBone {boneName=dyn_lace_r_01_01}
      boxExtents=<30 30 30>
      constraintSetup {angularLimitsMax=<0 2 25>, angularLimitsMin=<0 -20 -25>, angularTarget=<0 0 0>}
      localJointOffset=<0.5 0 0>
     }]
    relativeSpaceBone {boneName=foot_l}
   }
   n78 = Anim Dynamics - Bone: dyn_lace_r_01_02(ComponentPose=@n77, Alpha=1, ExternalForce="0,0,0") {
    angularDampingOverride=0.9
    angularSpringConstant=0.99
    bAngularSpring=true
    bOverrideAngularDamping=true
    simulationSpace=RootRelative
    alphaBoolBlend {blendOption=Linear}
    boundBone {boneName=dyn_lace_r_01_02}
    componentAppliedLinearAccClamp=<10000 10000 10000>
    componentLinearAccScale=<3 3 3>
    componentLinearVelScale=<3 3 3>
    physicsBodyDefinitions [{
      boundBone {boneName=dyn_lace_r_01_02}
      boxExtents=<15 15 15>
      constraintSetup {angularLimitsMax=<0 2 25>, angularLimitsMin=<0 -20 -25>, angularTarget=<0 0 0>}
      localJointOffset=<0.5 0 0>
     }]
    relativeSpaceBone {boneName=foot_l}
   }
   n79 = Anim Dynamics - Bone: dyn_lace_r_01_03(ComponentPose=@n78, Alpha=1, ExternalForce="0,0,0") {
    angularDampingOverride=0.8
    angularSpringConstant=0.99
    bAngularSpring=true
    bOverrideAngularDamping=true
    simulationSpace=RootRelative
    alphaBoolBlend {blendOption=Linear}
    boundBone {boneName=dyn_lace_r_01_03}
    componentAppliedLinearAccClamp=<10000 10000 10000>
    componentLinearAccScale=<2 2 2>
    componentLinearVelScale=<2 2 2>
    physicsBodyDefinitions [{
      boundBone {boneName=dyn_lace_r_01_03}
      boxExtents=<10 10 10>
      constraintSetup {angularLimitsMax=<0 2 25>, angularLimitsMin=<0 -20 -25>, angularTarget=<0 0 0>}
      localJointOffset=<0.5 0 0>
     }]
    relativeSpaceBone {boneName=foot_l}
   }
   n80 = "Constraint - dyn_lace_r_02_mstr (PPP)"!(ComponentPose=@n79, Alpha=1, ConstraintWeights_0=0.02, ConstraintWeights_1:float=0.03, ConstraintWeights_2:float=0.95) -> (Pose) {
    alphaBoolBlend {blendOption=Linear}
    boneToModify {boneName=dyn_lace_r_02_mstr}
    constraintSetup [{targetBone {boneName=calf_twist_01_r}}, {targetBone {boneName=calf_twist_02_r}}, {targetBone {boneName=foot_r}}]
    constraintWeights [1,1,1]
   }
   n81 = Anim Dynamics - Bone: dyn_lace_r_02_01(ComponentPose=@n80, Alpha=1, ExternalForce="0,0,0") {
    angularDampingOverride=0.9
    angularSpringConstant=0.9999
    bOverrideAngularDamping=true
    bUseGravityOverride=true
    simulationSpace=RootRelative
    alphaBoolBlend {blendOption=Linear}
    boundBone {boneName=dyn_lace_r_02_01}
    componentAppliedLinearAccClamp=<1000 1000 1000>
    componentLinearAccScale=<5 5 5>
    componentLinearVelScale=<5 5 5>
    gravityOverride {z=-980}
    physicsBodyDefinitions [{
      boundBone {boneName=dyn_lace_r_02_01}
      boxExtents=<30 30 30>
      constraintSetup {angularLimitsMax=<0 0 90>, angularLimitsMin=<0 0 -90>, angularTarget=<0 0 0>}
      localJointOffset=<0.3 0 0>
     }]
    planarLimits [{drivingBone {boneName=foot_r}, planeTransform {location=<10 0 0>, rotation=<rot 20 180 180>}}]
    relativeSpaceBone {boneName=foot_l}
   }
   n82 = Anim Dynamics - Bone: dyn_lace_r_02_02(ComponentPose=@n81, Alpha=1, ExternalForce="0,0,0") {
    angularDampingOverride=0.9
    angularSpringConstant=0.9999
    bOverrideAngularDamping=true
    bUseGravityOverride=true
    simulationSpace=RootRelative
    alphaBoolBlend {blendOption=Linear}
    boundBone {boneName=dyn_lace_r_02_02}
    componentAppliedLinearAccClamp=<1000 1000 1000>
    componentLinearAccScale=<3 3 3>
    componentLinearVelScale=<3 3 3>
    gravityOverride {z=-980}
    physicsBodyDefinitions [{
      boundBone {boneName=dyn_lace_r_02_02}
      boxExtents=<15 15 15>
      constraintSetup {angularLimitsMax=<0 0 90>, angularLimitsMin=<0 0 -90>, angularTarget=<0 0 0>}
      localJointOffset=<0.3 0 0>
     }]
    planarLimits [{drivingBone {boneName=foot_r}, planeTransform {location=<10 0 0>, rotation=<rot 20 180 180>}}]
    relativeSpaceBone {boneName=foot_l}
   }
   n83 = Anim Dynamics - Bone: dyn_lace_r_02_03(ComponentPose=@n82, Alpha=1, ExternalForce="0,0,0") {
    angularDampingOverride=0.8
    angularSpringConstant=0.9999
    bOverrideAngularDamping=true
    bUseGravityOverride=true
    simulationSpace=RootRelative
    alphaBoolBlend {blendOption=Linear}
    boundBone {boneName=dyn_lace_r_02_03}
    componentAppliedLinearAccClamp=<1000 1000 1000>
    componentLinearAccScale=<2 2 2>
    componentLinearVelScale=<2 2 2>
    gravityOverride {z=-980}
    physicsBodyDefinitions [{
      boundBone {boneName=dyn_lace_r_02_03}
      boxExtents=<10 10 10>
      constraintSetup {angularLimitsMax=<0 0 90>, angularLimitsMin=<0 0 -90>, angularTarget=<0 0 0>}
      localJointOffset=<0.5 0 0>
     }]
    planarLimits [{drivingBone {boneName=foot_r}, planeTransform {location=<10 0 0>, rotation=<rot 20 180 180>}}]
    relativeSpaceBone {boneName=foot_l}
   }
   Output Pose(Result=Component To Local(ComponentPose=@n83)) {layerGroup=None}
  ]}
 EventGraph {nodes [
   Event BlueprintUpdateAnimation()
   n1 = BreakTransform(InTransform=GetDeltaTransformFromRefPose(self=GetOwnerSkeletalMeshComponent(), BoneName=dyn_necklace_02, BaseName=spine_05))
   > Set necklace alpha(necklace alpha:double=MapRangeClamped(Value=BreakRotator(InRot=@n1.Rotation).Pitch, InRangeA=-10, InRangeB=-40)) -> (Output_Get:double)
   > n2 = Gravity Override State Function DEPRECATED(CharacterPartAnimInstance=Self-Reference(), GameTimeInSeconds=GetGameTimeInSeconds(), GravityOverrideParams=Get Gravity Override Params())
   > Set GravityOverride(GravityOverride=@n2)
   > Set FrontEndAlpha(FrontEndAlpha:double=SelectFloat(A=0, B=0.7, bPickA=Get bIsInFrontEnd())) -> (Output_Get:double)
  ]}
]
variables [
 FrontEndAlpha {type=number}
 Gravity Override Params {type=GravityOverrideParamsStruct, default {skydive_GoingUp {gravityOverride=<0 -300 -900>}}}
 GravityOverride {type=Vector}
 necklace alpha {type=number}
]
